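# Control Rig module that drives a five-bone hind leg from a four-bone driver leg.
# Construction resolves the Driver Tip and Driven Tip connectors and walks parents to cache both chains.
# Each update, rotation-only parent constraints map driver thigh, calf and foot onto driven thigh, calf and ball.
# FABRIK then pulls the driven chain to the driver ball plus the toe's initial offset, and the toe is set to that transform.
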ControlRigBlueprint /EpicControlRigRuntime/CRM_Runtime_HindLeg_v01
graphs [
 ConstructionGraph {nodes [
   RigUnit_PrepareForExecution = &RigUnit_PrepareForExecution()
   RigVMFunction_Sequence = &RigVMFunction_Sequence(ExecuteContext=@RigUnit_PrepareForExecution.ExecutePin)
   n1 = /Script/ControlRig.RigUnit_ResolveConnector(Connector=(Type=Connector,Name="Driven Tip"), SkipSocket=false) {is=struct RigVMUnitNode}
   RigVMVariableNode(Value=@n1.Result, ExecuteContext=@RigVMFunction_Sequence.B) {variable=Driven Toe}
   n2 = /Script/ControlRig.RigUnit_HierarchyGetParent(bDefaultParent=True, Child=@n1.Result) {is=struct RigVMUnitNode}
   > RigVMVariableNode(Value=@n2.Parent) {variable=Driven Ball}
   n3 = &RigUnit_HierarchyGetParent(bDefaultParent=True, Child=@n2.Parent)
   > RigVMVariableNode(Value=@n3.Parent) {variable=Driven Foot}
   n4 = &RigUnit_HierarchyGetParent(bDefaultParent=True, Child=@n3.Parent)
   > RigVMVariableNode(Value=@n4.Parent) {variable=Driven Calf}
   > RigVMVariableNode(Value=&RigUnit_HierarchyGetParent(bDefaultParent=True, Child=@n4.Parent).Parent) {variable=Driven Thigh}
   n5 = &RigUnit_ResolveConnector(Connector=(Type=Connector,Name="Driver Tip"), SkipSocket=false)
   RigVMVariableNode(Value=@n5.Result, ExecuteContext=@RigVMFunction_Sequence.A) {variable=Driver Ball}
   n6 = &RigUnit_HierarchyGetParent(bDefaultParent=True, Child=@n5.Result)
   > RigVMVariableNode(Value=@n6.Parent) {variable=Driver Foot}
   n7 = &RigUnit_HierarchyGetParent(bDefaultParent=True, Child=@n6.Parent)
   > RigVMVariableNode(Value=@n7.Parent) {variable=Driver Calf}
   > RigVMVariableNode(Value=&RigUnit_HierarchyGetParent(bDefaultParent=True, Child=@n7.Parent).Parent) {variable=Driver Thigh}
  ]}
 RigVMFunctionLibrary {links [], nodes []}
 RigVMModel {nodes [
   RigUnit_BeginExecution = &RigUnit_BeginExecution()
   &RigUnit_ParentConstraint(
    AdvancedSettings=(InterpolationType=Average,RotationOrderForFilter=XZY)
    Filter=(TranslationFilter=(bX=false,bY=false,bZ=false),RotationFilter=(bX=True,bY=True,bZ=True),ScaleFilter=(bX=false,bY=false,bZ=false))
    Parents=((Item=(Type=Bone,Name="thigh_l"),Weight=1))
    Weight=1
    ExecutePin=@RigUnit_BeginExecution.ExecutePin
    Child=RigVMVariableNode() {variable=Driven Thigh}.Value
    Parents.0.Item=RigVMVariableNode() {variable=Driver Thigh}.Value
   )
   > &RigUnit_ParentConstraint(
    AdvancedSettings=(InterpolationType=Average,RotationOrderForFilter=XZY)
    Filter=(TranslationFilter=(bX=false,bY=false,bZ=false),RotationFilter=(bX=True,bY=True,bZ=True),ScaleFilter=(bX=false,bY=false,bZ=false))
    Parents=((Item=(Type=Bone,Name="calf_l"),Weight=1))
    Weight=1
    Child=RigVMVariableNode() {variable=Driven Calf}.Value
    Parents.0.Item=RigVMVariableNode() {variable=Driver Calf}.Value
   )
   > &RigUnit_ParentConstraint(
    AdvancedSettings=(InterpolationType=Average,RotationOrderForFilter=XZY)
    Filter=(TranslationFilter=(bX=false,bY=false,bZ=false),RotationFilter=(bX=True,bY=True,bZ=True),ScaleFilter=(bX=false,bY=false,bZ=false))
    Parents=((Item=(Type=Bone,Name="foot_l"),Weight=1))
    Weight=1
    Child=RigVMVariableNode() {variable=Driven Ball}.Value
    Parents.0.Item=RigVMVariableNode() {variable=Driver Foot}.Value
   )
   n1 = RigVMVariableNode() {variable=Driver Ball}
   RigUnit_GetRelativeTransformForItem = &RigUnit_GetRelativeTransformForItem(bChildInitial=true, bParentInitial=true, Child=RigVMVariableNode() {variable=Driven Toe}.Value, Parent=@n1.Value)
   RigUnit_GetTransform = &RigUnit_GetTransform(bInitial=False, Item=@n1.Value)
   n2 = &RigVMFunction_MathTransformMul(A=@RigUnit_GetRelativeTransformForItem.RelativeTransform, B=@RigUnit_GetTransform.Transform) {template="Multiply::Execute(in A,in B,out Result)"}
   RigUnit_ItemArray = &RigUnit_ItemArray(
    Items=((Type=Bone,Name="hind_thigh_l"),(Type=Bone,Name="hind_calf_l"),(Type=Bone,Name="hind_foot_l"),(Type=Bone,Name="hind_ball_l"),(Type=Bone,Name="hind_toe_l"))
    Items.1=RigVMVariableNode() {variable=Driven Calf}.Value
    Items.2=RigVMVariableNode() {variable=Driven Foot}.Value
    Items.3=RigVMVariableNode() {variable=Driven Ball}.Value
    Items.4=RigVMVariableNode() {variable=Driven Toe}.Value
    Items.0=RigVMVariableNode() {variable=Driven Thigh}.Value
   )
   > /Script/ControlRig.RigUnit_FABRIKItemArray(
    Precision=1
    Weight=1
    bPropagateToChildren=True
    bSetEffectorTransform=True
    EffectorTransform=@n2.Result
    Items=@RigUnit_ItemArray.Items
    MaxIterations=RigVMVariableNode() {variable=FABRIK Max Iterations}.Value
   ) {is=struct RigVMUnitNode}
   > RigUnit_SetTransform = &RigUnit_SetTransform(Weight=1, bInitial=False, Item=RigVMVariableNode() {variable=Driven Toe}.Value, Value=@n2.Result)
  ]}
]
hierarchy [
 Driver Tip {type=Connector, loc=<0 0 0>, rot=<0 0 0 1>}
 Root {type=Socket, loc=<0 0 0>, rot=<0 0 0 1>}
 Driven Tip {type=Connector, loc=<0 0 0>, rot=<0 0 0 1>}
]
variables [
 Driven Ball {type=FRigElementKey, default=(Type=Connector,Name="")}
 Driven Calf {type=FRigElementKey, default=(Type=Connector,Name="")}
 Driven Foot {type=FRigElementKey, default=(Type=Connector,Name="")}
 Driven Thigh {type=FRigElementKey, default=(Type=Connector,Name="")}
 Driven Toe {type=FRigElementKey, default=(Type=Connector,Name="")}
 Driver Ball {type=FRigElementKey, default=(Type=Connector,Name="")}
 Driver Calf {type=FRigElementKey, default=(Type=Connector,Name="")}
 Driver Foot {type=FRigElementKey, default=(Type=Connector,Name="")}
 Driver Thigh {type=FRigElementKey, default=(Type=Connector,Name="")}
 FABRIK Max Iterations {type=int32, default=10}
]
