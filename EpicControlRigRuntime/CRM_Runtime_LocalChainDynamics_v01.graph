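# Control Rig module that adds local dynamics to a bone chain, with optional procedural "Keep Alive" motion for characters.
# Construction collects the Start Bone connector's bone and descendants into Chain for the Local Chain Dynamics function.
# Beforehand, if Enable Motion is set and Base Motion Bone is connected, Keep Alive rotates that bone on Y and Z.
# Looped curves blend with Kalman-smoothed noise, weighted by a spring interpolation driven by zero Pawn Velocity.
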ControlRigBlueprint /EpicControlRigRuntime/CRM_Runtime_LocalChainDynamics_v01
graphs [
 New Function_ContainedGraph {nodes [
   RigVMFunctionEntryNode()
   n1 = /Script/RigVM.RigVMFunction_TimeLoop(Normalize=true, Speed=RigVMVariableNode() {variable=Speed Y}.Value, Duration=RigVMVariableNode() {variable=Duration Y}.Value) {is=struct RigVMUnitNode}
   Noise = /Script/RigVM.RigVMFunction_NoiseDouble(
    Frequency=2
    Speed=0.015
    Value=@n1.Relative
    Minimum=RigVMVariableNode() {variable=Min Y}.Value
    Maximum=RigVMVariableNode() {variable=Max Y}.Value
   ) {is=struct RigVMUnitNode, template="Noise::Execute(in Value,in Speed,in Frequency,in Minimum,in Maximum,out Result)"}
   KalmanFilter = &RigVMFunction_KalmanFloat(BufferSize=16, Value=@Noise.Result) {template="KalmanFilter::Execute(in Value,in BufferSize,out Result)"}
   n2 = /Script/RigVM.RigVMFunction_AnimRichCurve(Curve=(EditorCurveData=(Keys=((InterpMode=RCIM_Cubic,TangentMode=RCTM_User),(InterpMode=RCIM_Cubic,TangentMode=RCTM_User,Time=0.5,Value=1),(InterpMode=RCIM_Cubic,TangentMode=RCTM_User,Time=1))))) {is=struct RigVMUnitNode}
   Interpolate = /Script/RigVM.RigVMFunction_MathDoubleLerp(
    B=@KalmanFilter.Result
    T=&RigVMFunction_NoiseDouble(Frequency=0.1, Maximum=1, Minimum=0, Speed=1, Value=@n1.Relative) {template="Noise::Execute(in Value,in Speed,in Frequency,in Minimum,in Maximum,out Result)"}.Result
    A=&RigVMFunction_AnimEvalRichCurve(
     SourceMinimum=0
     Curve=@n2.Curve
     Value=@n1.Relative
     TargetMinimum=RigVMVariableNode() {variable=Min Y}.Value
     TargetMaximum=RigVMVariableNode() {variable=Max Y}.Value
    ).Result
   ) {is=struct RigVMUnitNode, template="Interpolate::Execute(in A,in B,in T,out Result)"}
   n3 = &RigVMFunction_MathQuaternionToEuler(RotationOrder=ZYX, Value=(X=0,Y=0,Z=0,W=1))
   n4 = &RigVMFunction_MathDoubleAdd(B=@Interpolate.Result, A=@n3.Result.Y)
   n5 = RigVMVariableNode() {variable=Alpha}
   n6 = &RigVMFunction_TimeLoop(Normalize=true, Speed=RigVMVariableNode() {variable=Speed Z}.Value, Duration=RigVMVariableNode() {variable=Duration Z}.Value)
   n7 = &RigVMFunction_NoiseDouble(
    Frequency=1
    Speed=0.1
    Value=@n6.Relative
    Maximum=RigVMVariableNode() {variable=Max Z}.Value
    Minimum=RigVMVariableNode() {variable=Min Z}.Value
   ) {template="Noise::Execute(in Value,in Speed,in Frequency,in Minimum,in Maximum,out Result)"}
   n8 = RigVMVariableNode() {variable=Max Z}
   n9 = RigVMVariableNode() {variable=Min Z}
   n10 = &RigVMFunction_MathDoubleAdd(
    B=&RigVMFunction_MathDoubleLerp(
     B=&RigVMFunction_KalmanFloat(BufferSize=16, Value=@n7.Result) {template="KalmanFilter::Execute(in Value,in BufferSize,out Result)"}.Result
     T=&RigVMFunction_NoiseDouble(Frequency=0.2, Maximum=1, Minimum=0, Speed=1.1, Value=@n6.Relative) {template="Noise::Execute(in Value,in Speed,in Frequency,in Minimum,in Maximum,out Result)"}.Result
     A=&RigVMFunction_AnimEvalRichCurve(SourceMinimum=0, Curve=@n2.Curve, Value=@n6.Relative, TargetMaximum=@n8.Value, TargetMinimum=@n9.Value).Result
    ) {template="Interpolate::Execute(in A,in B,in T,out Result)"}.Result
    A=@n3.Result.Z
   )
   n11 = RigVMVariableNode() {variable=Alpha}
   > RigVMFunctionReturnNode(Rotation=&RigVMFunction_MathQuaternionFromEuler(
     Euler=(X=0,Y=0,Z=0)
     Euler.Y=&RigVMFunction_MathDoubleLerp(B=@n4.Result, A=@n3.Result.Y, T=@n5.Value) {template="Interpolate::Execute(in A,in B,in T,out Result)"}.Result
     Euler.Z=&RigVMFunction_MathDoubleLerp(B=@n10.Result, A=@n3.Result.Z, T=@n11.Value) {template="Interpolate::Execute(in A,in B,in T,out Result)"}.Result
     Euler.X=@n3.Result.X
    ).Result)
  ]}
 RigVMFunctionLibrary {nodes ["Keep Alive " = RigVMCollapseNode(Alpha=1, Duration Y=2, Duration Z=1, Max Y=10, Max Z=20, Min Y=-10, Min Z=5, Speed Y=0.25, Speed Z=0.25)]}
 RigVMModel {nodes [
   RigUnit_BeginExecution = &RigUnit_BeginExecution()
   RigVMFunction_Sequence = &RigVMFunction_Sequence(ExecuteContext=@RigUnit_BeginExecution.ExecutePin)
   Local Chain Dynamics = RigVMFunctionReferenceNode(
    ExecuteContext=@RigVMFunction_Sequence.B
    Bones=RigVMVariableNode() {variable=Chain}.Value
    Strength=RigVMVariableNode() {variable=Strength}.Value
    Damping=RigVMVariableNode() {variable=Damping}.Value
    Blend Start=RigVMVariableNode() {variable=Blend Start}.Value
    Blend End=RigVMVariableNode() {variable=Blend End}.Value
    Blend Twist=RigVMVariableNode() {variable=Blend Twist}.Value
    Draw Debug=RigVMVariableNode() {variable=Draw Debug}.Value
   )
   n1 = /Script/ControlRig.RigUnit_ResolveConnector(Connector=(Type=Connector,Name="Base Motion Bone"), SkipSocket=False) {is=struct RigVMUnitNode}
   RigVMFunction_MathBoolAnd = &RigVMFunction_MathBoolAnd(B=@n1.bIsConnected, A=RigVMVariableNode() {variable=Enable Motion}.Value)
   n2 = &RigVMFunction_ControlFlowBranch(Condition=@RigVMFunction_MathBoolAnd.Result, ExecuteContext=@RigVMFunction_Sequence.A)
   RigVMFunction_MathVectorLengthSquared = /Script/RigVM.RigVMFunction_MathVectorLengthSquared(Value=RigVMVariableNode() {variable=Pawn Velocity}.Value) {is=struct RigVMUnitNode, template="LengthSquared::Execute(in Value,out Result)"}
   LessEqual = /Script/RigVM.RigVMFunction_MathDoubleLessEqual(B=0, A=@RigVMFunction_MathVectorLengthSquared.Result) {is=struct RigVMUnitNode, template="LessEqual::Execute(in A,in B,out Result)"}
   n3 = &RigVMFunction_MathBoolToFloat(Value=@LessEqual.Result)
   SpringInterp = /Script/ControlRig.RigUnit_SpringInterpV2(
    CriticalDamping=1
    Force=0
    Target=1
    TargetVelocityAmount=0
    bInitializeFromTarget=true
    bUseCurrentInput=true
    Current=@n3.Result
    Strength=@n3.Result
   ) {
    is=struct RigVMUnitNode
    template="SpringInterp::Execute(in Target,in Strength,in CriticalDamping,in Force,in bUseCurrentInput,in Current,in TargetVelocityAmount,in bInitializeFromTarget,out Result,out Velocity)"
   }
   "Keep Alive " = RigVMFunctionReferenceNode(
    Alpha=@SpringInterp.Result
    Speed Y=RigVMVariableNode() {variable=Speed Y}.Value
    Duration Y=RigVMVariableNode() {variable=Duration Y}.Value
    Min Y=RigVMVariableNode() {variable=Min Y}.Value
    Max Y=RigVMVariableNode() {variable=Max Y}.Value
    Speed Z=RigVMVariableNode() {variable=Speed Z}.Value
    Duration Z=RigVMVariableNode() {variable=Duration Z}.Value
    Min Z=RigVMVariableNode() {variable=Min Z}.Value
    Max Z=RigVMVariableNode() {variable=Max Z}.Value
   )
   &RigUnit_ModifyTransforms(
    ItemToModify=((Item=(Type=Bone,Name="None"),Transform=(Rotation=(X=0,Y=0,Z=0,W=1),Translation=(X=0,Y=0,Z=0),Scale3D=(X=1,Y=1,Z=1))))
    Weight=1
    WeightMinimum=0
    ItemToModify.0.Transform.Rotation=@"Keep Alive ".Rotation
    ItemToModify.0.Item=@n1.Result
    ExecutePin=@n2.True
   )
  ]}
 RigVMModel Construction Graph {nodes [
   n1 = &RigUnit_PrepareForExecution()
   RigVMVariableNode(
    Value=/Script/ControlRig.RigUnit_CollectionChildrenArray(
     TypeToSearch=Bone
     bDefaultChildren=True
     bIncludeParent=true
     bRecursive=true
     Parent=&RigUnit_ResolveConnector(Connector=(Type=Connector,Name="Start Bone"), SkipSocket=false).Result
    ) {is=struct RigVMUnitNode}.Items
    ExecuteContext=@n1.ExecutePin
   ) {variable=Chain}
  ]}
]
hierarchy [
 Start Bone {type=Connector, loc=<0 0 0>, rot=<0 0 0 1>}
 Root {type=Socket, loc=<0 0 0>, rot=<0 0 0 1>}
 Base Motion Bone {type=Connector, loc=<0 0 0>, rot=<0 0 0 1>}
]
variables [
 Blend End {type=double, default=0.25}
 Blend Start {type=double, default=0.25}
 Blend Twist {type=double, default=0}
 Chain {type="TArray<FRigElementKey>"}
 Damping {type=double, default=0.6}
 Draw Debug {type=bool, default=False}
 Duration Y {type=double, default=2}
 Duration Z {type=double, default=1}
 Enable Motion {type=bool, default=False}
 Max Y {type=double, default=10}
 Max Z {type=double, default=10}
 Min Y {type=double, default=-10}
 Min Z {type=double, default=-10}
 Pawn Velocity {type=FVector, default=(X=0,Y=0,Z=0)}
 Speed Y {type=double, default=0.25}
 Speed Z {type=double, default=0.25}
 Strength {type=double, default=5}
]
