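# Modular Control Rig runtime module that applies a parent constraint to the element resolved from its Driven connector.
# Construction stores the resolved connector as the child, then each forward solve constrains it to the configurable parents with filter, offset and interpolation settings.
# The constraint weight is either the Weight variable alone or, with Use Metadata on, that weight multiplied by a named module metadata value that defaults to one.
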ControlRigBlueprint /EpicControlRigRuntime/CRM_Runtime_ParentConstraint_v01
graphs [
 ConstructionGraph {nodes [
   RigUnit_PrepareForExecution = &RigUnit_PrepareForExecution()
   RigVMVariableNode(
    Value=/Script/ControlRig.RigUnit_ResolveConnector(Connector=(Type=Connector,Name="Driven"), SkipSocket=False) {is=struct RigVMUnitNode}.Result
    ExecuteContext=@RigUnit_PrepareForExecution.ExecutePin
   ) {variable=Bone}
  ]}
 RigVMFunctionLibrary {links [], nodes []}
 RigVMModel {nodes [
   RigUnit_BeginExecution = &RigUnit_BeginExecution()
   n1 = &RigVMFunction_ControlFlowBranch(ExecuteContext=@RigUnit_BeginExecution.ExecutePin, Condition=RigVMVariableNode() {variable=Use Metadata}.Value)
   &RigUnit_ParentConstraint(
    ExecutePin=@n1.Completed
    Parents=RigVMVariableNode() {variable=Parents}.Value
    Filter=RigVMVariableNode() {variable=Filter}.Value
    Weight=RigVMVariableNode() {variable=Final Weight}.Value
    AdvancedSettings=RigVMVariableNode() {variable=Advanced Settings}.Value
    Child=RigVMVariableNode() {variable=Bone}.Value
    bMaintainOffset=RigVMVariableNode() {variable=Maintain Offset}.Value
   )
   RigVMVariableNode(ExecuteContext=@n1.False, Value=RigVMVariableNode() {variable=Weight}.Value) {variable=Final Weight}
   Get_Module_Metadata = RigVMDispatchNode(Default=1, Name=RigVMVariableNode() {variable=Metadata Name}.Value, NameSpace=RigVMVariableNode() {variable=Metadata Namespace}.Value) {template="DISPATCH_RigDispatch_GetModuleMetadata(in Name,in NameSpace,in Default,out Value,out Found)"}
   RigVMVariableNode(
    Value=&RigVMFunction_MathDoubleMul(B=@Get_Module_Metadata.Value, A=RigVMVariableNode() {variable=Weight}.Value).Result
    ExecuteContext=@n1.True
   ) {variable=Final Weight}
   RigVMVariableNode() {variable=Parents_Weights}
  ]}
]
hierarchy [
 Driven {type=Connector, loc=<0 0 0>, rot=<0 0 0 1>}
 Root {type=Socket, loc=<0 0 0>, rot=<0 0 0 1>}
]
variables [
 Advanced Settings {type=FRigUnit_ParentConstraint_AdvancedSettings, default=(InterpolationType=Average,RotationOrderForFilter=XZY)}
 Bone {type=FRigElementKey, default=(Type=Bone,Name="")}
 Filter {
  type=FTransformFilter
  default=(TranslationFilter=(bX=True,bY=True,bZ=True),RotationFilter=(bX=True,bY=True,bZ=True),ScaleFilter=(bX=True,bY=True,bZ=True))
 }
 Final Weight {type=double, default=0}
 Maintain Offset {type=bool, default=True}
 Metadata Name {type=FName, default=None}
 Metadata Namespace {type=ERigMetaDataNameSpace, default=None}
 Parents {type="TArray<FConstraintParent>", default=(())}
 Parents_Weights {type="TArray<FConstraintParent>", default=(())}
 Use Metadata {type=bool, default=False}
 Weight {type=float, default=1}
]
