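# Control Rig module that adds kinematic physics colliders to the bones resolved by its Collision Bones connector during construction.
# It reads the physics solver owner and name from parent module metadata, stores the solver key and republishes both on itself.
# Bones lacking "Has Physics Body" metadata get a body on that solver, then are tagged and recorded in Body Components.
# Collision shapes match by index, clamped so extra bones reuse the last entry.
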ControlRigBlueprint /EpicControlRigRuntime/CRM_Runtime_Physics_Add_Collider_v01
graphs [
 RigVMFunctionLibrary {links [], nodes []}
 RigVMModel {nodes [
   n1 = &RigUnit_PrepareForExecution()
   RigVMFunction_Sequence = &RigVMFunction_Sequence(ExecuteContext=@n1.ExecutePin)
   RigVMDispatchNode(ExecuteContext=@RigVMFunction_Sequence.B, Array=RigVMVariableNode() {variable=Body Components}.Value) {template="DISPATCH_RigVMDispatch_ArrayReset(io Array)"}
   > For_Each = RigVMDispatchNode(Array=/Script/ControlRig.RigUnit_ResolveArrayConnector(Connector=(Type=Connector,Name="Collision Bones"), SkipSocket=true) {is=struct RigVMUnitNode}.Result) {template="DISPATCH_RigVMDispatch_ArrayIterator(in Array,out Element,out Index,out Count,out Ratio)"}
   Get_Metadata = RigVMDispatchNode(Default=None, Name=Has Physics Body, NameSpace=None, Item=@For_Each.Element) {template="DISPATCH_RigDispatch_GetMetadata(in Item,in Name,in NameSpace,in Default,out Value,out Found)"}
   RigVMFunction_ControlFlowBranch = &RigVMFunction_ControlFlowBranch(ExecuteContext=@For_Each.ExecuteContext, Condition=@Get_Metadata.Found)
   n2 = RigVMVariableNode() {variable=Collisions}
   Num = RigVMDispatchNode(Array=@n2.Value) {template="DISPATCH_RigVMDispatch_ArrayGetNum(in Array,out Num)"}
   Greater = /Script/RigVM.RigVMFunction_MathIntGreater(B=0, A=@Num.Num) {is=struct RigVMUnitNode, template="Greater::Execute(in A,in B,out Result)"}
   Subtract = /Script/RigVM.RigVMFunction_MathIntSub(B=1, A=@Num.Num) {is=struct RigVMUnitNode, template="Subtract::Execute(in A,in B,out Result)"}
   Clamp = &RigVMFunction_MathIntClamp(
    Value=@For_Each.Index
    Maximum=RigVMDispatchNode(False=0, Condition=@Greater.Result, True=@Subtract.Result) {template="DISPATCH_RigVMDispatch_If(in Condition,in True,in False,out Result)"}.Result
   )
   n3 = /Script/ControlRigPhysics.RigUnit_AddPhysicsBody(
    BodyData=(MovementType=Kinematic,CollisionType=QueryAndPhysics,KinematicTargetSpace=OffsetInBoneSpace,GravityMultiplier=1,PhysicsBlendWeight=1,bUpdateKinematicFromSimulation=true,bEnableCCD=False)
    Dynamics="(Density=1,MassOverride=1,bOverrideCentreOfMass=False,CentreOfMassOverride=(X=0,Y=0,Z=0),CentreOfMassNudge=(X=0,Y=0,Z=0),bOverrideMomentsOfInertia=False,MomentsOfInertiaOverride=(X=1,Y=1,Z=1),LinearDamping=0,AngularDamping=0,bScaleDampingByInverseMass=Fals ...(truncated)"
    Solver=(PhysicsSolverComponentKey=(ElementKey=(Type=Bone,Name="None"),Name="PhysicsBody"),bUseAutomaticSolver=false,SourceBone=(Type=None,Name="None"),TargetBone=(Type=None,Name="None"),bIncludeInChecksForReset=false)
    Collision=RigVMDispatchNode(Index=@Clamp.Result, Array=@n2.Value) {template="DISPATCH_RigVMDispatch_ArrayGetAtIndex(in Array,in Index,out Element)"}.Element
    Owner=@For_Each.Element
    ExecutePin=@RigVMFunction_ControlFlowBranch.False
    Solver.PhysicsSolverComponentKey=RigVMVariableNode() {variable=Solver}.Value
   ) {is=struct RigVMUnitNode}
   > RigVMDispatchNode(Element=@n3.PhysicsBodyComponentKey, Array=RigVMVariableNode() {variable=Body Components}.Value) {template="DISPATCH_RigVMDispatch_ArrayAdd(io Array,in Element,out Index)"}
   > Set_Metadata_1 = RigVMDispatchNode(Name=Has Physics Body, NameSpace=None, Item=@n3.PhysicsBodyComponentKey.ElementKey, Value=@n3.PhysicsBodyComponentKey.Name) {template="DISPATCH_RigDispatch_SetMetadata(in Item,in Name,in NameSpace,in Value,out Success)"}
   Get_Module_Metadata = RigVMDispatchNode(Default=(Type=Bone,Name="None"), Name=Physics Solver Owner, NameSpace=Parent) {template="DISPATCH_RigDispatch_GetModuleMetadata(in Name,in NameSpace,in Default,out Value,out Found)"}
   Get_Module_Metadata_1 = RigVMDispatchNode(Default=PhysicsSolver, Name=Physics Solver Name, NameSpace=Parent) {template="DISPATCH_RigDispatch_GetModuleMetadata(in Name,in NameSpace,in Default,out Value,out Found)"}
   RigVMVariableNode(
    Value=(ElementKey=(Type=None,Name="None"),Name="None")
    Value.ElementKey=@Get_Module_Metadata.Value
    Value.Name=@Get_Module_Metadata_1.Value
    ExecuteContext=@RigVMFunction_Sequence.A
   ) {variable=Solver}
   RigVMFunction_MathBoolAnd = &RigVMFunction_MathBoolAnd(A=@Get_Module_Metadata.Found, B=@Get_Module_Metadata_1.Found)
   > RigVMVariableNode(Value=@RigVMFunction_MathBoolAnd.Result) {variable=Solver Exists}
   > Set_Module_Metadata = RigVMDispatchNode(Name=Physics Solver Owner, NameSpace=Self, Value=@Get_Module_Metadata.Value) {template="DISPATCH_RigDispatch_SetModuleMetadata(in Name,in NameSpace,in Value,out Success)"}
   > Set_Module_Metadata_1 = RigVMDispatchNode(Name=Physics Solver Name, NameSpace=Self, Value=@Get_Module_Metadata_1.Value) {template="DISPATCH_RigDispatch_SetModuleMetadata(in Name,in NameSpace,in Value,out Success)"}
  ]}
]
hierarchy [
 Root {type=Connector, loc=<0 0 0>, rot=<0 0 0 1>}
 Root {type=Socket, loc=<0 0 0>, rot=<0 0 0 1>}
 Collision Bones {type=Connector, loc=<0 0 0>, rot=<0 0 0 1>}
]
variables [
 Body Components {type="TArray<FRigComponentKey>", default=(())}
 Body Components Settings {type="TArray<FRigPhysicsCollision>"}
 Collisions {type="TArray<FRigPhysicsCollision>", default=(())}
 Solver {type=FRigComponentKey, default=(ElementKey=(Type=None,Name=""),Name="")}
 Solver Exists {type=bool, default=False}
]
