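# Runtime physics Control Rig module that adds a physics solver component to the rig element resolved through its Root connector.
# Construction runs only when that connector is connected, passing the exposed solver and simulation space settings to the new solver.
# It then publishes the solver's owner key and name as module metadata in the Self namespace.
# The forward solve is empty, and the Solver Name variable is unused in the graphs.
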ControlRigBlueprint /EpicControlRigRuntime/CRM_Runtime_Physics_Add_Solver_v01
graphs [
 ConstructionGraph {nodes [
   RigUnit_PrepareForExecution = &RigUnit_PrepareForExecution()
   n1 = /Script/ControlRig.RigUnit_ResolveConnector(Connector=(Type=Connector,Name="Root"), SkipSocket=False) {is=struct RigVMUnitNode}
   RigVMFunction_ControlFlowBranch = &RigVMFunction_ControlFlowBranch(Condition=@n1.bIsConnected, ExecuteContext=@RigUnit_PrepareForExecution.ExecutePin)
   n2 = /Script/ControlRigPhysics.RigUnit_AddPhysicsSolver(
    Owner=@n1.Result
    ExecutePin=@RigVMFunction_ControlFlowBranch.True
    SolverSettings=RigVMVariableNode() {variable=Solver Settings}.Value
    SimulationSpaceSettings=RigVMVariableNode() {variable=Simulation Space Settings}.Value
   ) {is=struct RigVMUnitNode}
   > Set_Module_Metadata = RigVMDispatchNode(Name=Physics Solver Owner, NameSpace=Self, Value=@n2.PhysicsSolverComponentKey.ElementKey) {template="DISPATCH_RigDispatch_SetModuleMetadata(in Name,in NameSpace,in Value,out Success)"}
   > Set_Module_Metadata_1 = RigVMDispatchNode(Name=Physics Solver Name, NameSpace=Self, Value=@n2.PhysicsSolverComponentKey.Name) {template="DISPATCH_RigDispatch_SetModuleMetadata(in Name,in NameSpace,in Value,out Success)"}
  ]}
 RigVMFunctionLibrary {links [], nodes []}
 RigVMModel {nodes [RigUnit_BeginExecution = &RigUnit_BeginExecution()]}
]
hierarchy [
 Root {type=Connector, loc=<0 0 0>, rot=<0 0 0 1>}
 Root {type=Socket, loc=<0 0 0>, rot=<0 0 0 1>}
]
variables [
 Simulation Space Settings {
  type=FRigPhysicsSimulationSpaceSettings
  default="(SpaceMovementAmount=1,VelocityScaleZ=1,bClampLinearVelocity=True,MaxLinearVelocity=700,bClampAngularVelocity=True,MaxAngularVelocity=12,bClampLinearAcceleration=True,MaxLinearAcceleration=2800,bClampAngularAcceleration=True,MaxAngularAcceleration=48,Linea ...(truncated)"
 }
 Solver Name {type=FName, default=ModularPhysicsSolver}
 Solver Settings {
  type=FRigPhysicsSolverSettings
  default="(bAutomaticallyAddPhysicsBodyComponents=False,SimulationSpace=Component,CollisionSpace=Component,SpaceBone=(Type=Bone,Name=\"\"),Collision=(Boxes=,Spheres=,Capsules=,Convexes=,Material=(Friction=1,Restitution=1,FrictionCombineMode=Multiply,RestitutionCombine ...(truncated)"
 }
]
