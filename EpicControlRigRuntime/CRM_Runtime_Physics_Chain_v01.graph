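# Control Rig module that builds a runtime physics chain between Start and End connectors, or on Start alone.
# Construction reads the parent module's solver, adds a kinematic body on Start's parent, then simulated bodies, joints and sim-space controls per element.
# Ramps over chain ratio shape body, joint and control settings, mass halves after the first element, and adjacent bodies ignore collision.
# Components are tagged as metadata; forward solve only applies a ramped rotation offset.
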ControlRigBlueprint /EpicControlRigRuntime/CRM_Runtime_Physics_Chain_v01
graphs [
 ConstructionGraph {nodes [
   RigUnit_PrepareForExecution = &RigUnit_PrepareForExecution()
   RigVMFunction_Sequence = RigVMAggregateNode(ExecuteContext=@RigUnit_PrepareForExecution.ExecutePin)
   n1 = /Script/ControlRigPhysics.RigUnit_AddPhysicsBody(
    BodyData=(MovementType=Kinematic,CollisionType=QueryAndPhysics,KinematicTargetSpace=OffsetInBoneSpace,GravityMultiplier=1,PhysicsBlendWeight=1,bUpdateKinematicFromSimulation=True,bEnableCCD=False)
    Collision=(Material=(Friction=0.5,Restitution=0,FrictionCombineMode=Multiply,RestitutionCombineMode=Multiply))
    Dynamics="(Density=1,MassOverride=1,bOverrideCentreOfMass=False,CentreOfMassOverride=(X=0,Y=0,Z=0),CentreOfMassNudge=(X=0,Y=0,Z=0),bOverrideMomentsOfInertia=False,MomentsOfInertiaOverride=(X=1,Y=1,Z=1),LinearDamping=0,AngularDamping=0,bScaleDampingByInverseMass=Fals ...(truncated)"
    Solver=(PhysicsSolverComponentKey=(ElementKey=(Type=Bone,Name="None"),Name="PhysicsBody"),bUseAutomaticSolver=false,SourceBone=(Type=None,Name="None"),TargetBone=(Type=None,Name="None"),bIncludeInChecksForReset=True)
    ExecutePin=@RigVMFunction_Sequence.C
    Owner=RigVMVariableNode() {variable=Kinematic Parent}.Value
    Solver.PhysicsSolverComponentKey=RigVMVariableNode() {variable=Solver}.Value
   ) {is=struct RigVMUnitNode}
   > RigVMVariableNode(Value=@n1.PhysicsBodyComponentKey) {variable=collision body}
   > /Script/ControlRigPhysics.RigUnit_HierarchyAutoCalculateCollision(MinAspectRatio=0.5, MinSize=0, PhysicsBodyComponentKey=@n1.PhysicsBodyComponentKey) {is=struct RigVMUnitNode}
   > Set_Metadata_1 = RigVMDispatchNode(Name=Has Physics Body, NameSpace=Self, Item=@n1.PhysicsBodyComponentKey.ElementKey, Value=@n1.PhysicsBodyComponentKey.Name) {template="DISPATCH_RigDispatch_SetMetadata(in Item,in Name,in NameSpace,in Value,out Success)"}
   RigVMDispatchNode(ExecuteContext=@RigVMFunction_Sequence.D, Array=RigVMVariableNode() {variable=Body Components}.Value) {template="DISPATCH_RigVMDispatch_ArrayReset(io Array)"}
   > RigVMVariableNode(Value=RigVMVariableNode() {variable=Start Mass}.Value) {variable=Current Mass}
   > For_Each = RigVMDispatchNode(Array=RigVMVariableNode() {variable=Chain}.Value) {template="DISPATCH_RigVMDispatch_ArrayIterator(in Array,out Element,out Index,out Count,out Ratio)"}
   > RigVMVariableNode(Value=@For_Each.Ratio) {variable=Chain Ratio}
   Curve Gain = RigVMFunctionReferenceNode(
    Value=RigVMVariableNode() {variable=Density}.Value
    Curve=RigVMVariableNode() {variable=Density Ramp}.Value
    Ratio=RigVMVariableNode() {variable=Chain Ratio}.Value
   )
   Curve Gain_1 = RigVMFunctionReferenceNode(
    Curve=RigVMVariableNode() {variable=Angular Damping Ramp}.Value
    Ratio=RigVMVariableNode() {variable=Chain Ratio}.Value
    Value=RigVMVariableNode() {variable=Angular Damping}.Value
   )
   Curve Gain_2 = RigVMFunctionReferenceNode(
    Value=RigVMVariableNode() {variable=Gravity Multiplier}.Value
    Curve=RigVMVariableNode() {variable=Gravity Ramp}.Value
    Ratio=RigVMVariableNode() {variable=Chain Ratio}.Value
   )
   Curve Gain_4 = RigVMFunctionReferenceNode(
    Value=RigVMVariableNode() {variable=Swing 1 Limit}.Value
    Curve=RigVMVariableNode() {variable=Swing 1 Limit Ramp}.Value
    Ratio=RigVMVariableNode() {variable=Chain Ratio}.Value
   )
   Curve Gain_5 = RigVMFunctionReferenceNode(
    Ratio=RigVMVariableNode() {variable=Chain Ratio}.Value
    Value=RigVMVariableNode() {variable=Swing 2 Limit}.Value
    Curve=RigVMVariableNode() {variable=Swing 2 Limit Ramp}.Value
   )
   Curve Gain_6 = RigVMFunctionReferenceNode(
    Ratio=RigVMVariableNode() {variable=Chain Ratio}.Value
    Value=RigVMVariableNode() {variable=Twist Limit}.Value
    Curve=RigVMVariableNode() {variable=Twist Limit Ramp}.Value
   )
   Curve Gain_7 = RigVMFunctionReferenceNode(
    Value=RigVMVariableNode() {variable=Angular Strength}.Value
    Curve=RigVMVariableNode() {variable=Angular Strength Ramp}.Value
    Ratio=RigVMVariableNode() {variable=Chain Ratio}.Value
   )
   Divide = &RigVMFunction_MathDoubleDiv(
    B=RigVMDispatchNode(
     False=2
     True=1
     Condition=RigVMDispatchNode(B=0, A=@For_Each.Index) {template="DISPATCH_RigVMDispatch_CoreEquals(in A,in B,out Result)"}.Result
    ) {template="DISPATCH_RigVMDispatch_If(in Condition,in True,in False,out Result)"}.Result
    A=RigVMVariableNode() {variable=Current Mass}.Value
   )
   Curve Gain_0 = RigVMFunctionReferenceNode(Value=1, Ratio=RigVMVariableNode() {variable=Chain Ratio}.Value, Curve=RigVMVariableNode() {variable=Cone Rotation Offset Ramp}.Value)
   RigVMFunction_MathQuaternionMake = /Script/RigVM.RigVMFunction_MathQuaternionMake(W=0, X=1, Y=0, Z=0) {is=struct RigVMUnitNode}
   Interpolate = /Script/RigVM.RigVMFunction_MathQuaternionSlerp(T=@"Curve Gain_0".Result, A=@RigVMFunction_MathQuaternionMake.Result, B=RigVMVariableNode() {variable=Cone Rotation Offset}.Value) {is=struct RigVMUnitNode, template="Interpolate::Execute(in A,in B,in T,out Result)"}
   Curve Gain_3 = RigVMFunctionReferenceNode(Value=1, Curve=RigVMVariableNode() {variable=Centre Of Mass Override Ramp}.Value, Ratio=RigVMVariableNode() {variable=Chain Ratio}.Value)
   RigVMFunction_MathVectorMake = /Script/RigVM.RigVMFunction_MathVectorMake(X=0, Y=0, Z=0) {is=struct RigVMUnitNode}
   > n2 = /Script/ControlRigPhysics.RigUnit_AddPhysicsComponents(
    BodyData=(MovementType=Simulated,CollisionType=QueryAndPhysics,GravityMultiplier=1,PhysicsBlendWeight=1,bUseSkeletalAnimation=True,bUpdateKinematicFromSimulation=True,KinematicTargetSpace=OffsetInBoneSpace,bEnableCCD=False)
    Collision="(Capsules=((RestOffset=0,Name=\"None\",bContributeToMass=True,TM=(Rotation=(X=0,Y=0,Z=0,W=1),Translation=(X=0,Y=0,Z=0),Scale3D=(X=1,Y=1,Z=1)),Radius=10,Length=10)),Material=(Friction=0.5,Restitution=0,FrictionCombineMode=Multiply,RestitutionCombineMode=Multi ...(truncated)"
    DriveData="(LinearDriveConstraint=(PositionTarget=(X=0,Y=0,Z=0),VelocityTarget=(X=0,Y=0,Z=0),XDrive=(Stiffness=50,Damping=1,MaxForce=0,bEnablePositionDrive=False,bEnableVelocityDrive=False),YDrive=(Stiffness=50,Damping=1,MaxForce=0,bEnablePositionDrive=False,bEnableV ...(truncated)"
    Dynamics="(Density=1,MassOverride=1,bOverrideCentreOfMass=False,CentreOfMassOverride=(X=0,Y=0,Z=0),CentreOfMassNudge=(X=0,Y=0,Z=0),bOverrideMomentsOfInertia=False,MomentsOfInertiaOverride=(X=1,Y=1,Z=1),LinearDamping=0,AngularDamping=0,bScaleDampingByInverseMass=Fals ...(truncated)"
    JointData="(bEnabled=True,bAutoCalculateParentOffset=true,ExtraParentOffset=(Rotation=(X=0,Y=0,Z=0,W=1),Translation=(X=0,Y=0,Z=0),Scale3D=(X=1,Y=1,Z=1)),bAutoCalculateChildOffset=True,ExtraChildOffset=(Rotation=(X=0,Y=0,Z=0,W=1),Translation=(X=0,Y=0,Z=0),Scale3D=(X=1 ...(truncated)"
    ParentSpaceControlData="(LinearStrength=0,LinearDampingRatio=1,LinearExtraDamping=0,MaxForce=0,AngularStrength=0,AngularDampingRatio=1,AngularExtraDamping=0,MaxTorque=0,LinearTargetVelocityMultiplier=1,AngularTargetVelocityMultiplier=1,SkeletalAnimationVelocityMultiplier=1,Custom ...(truncated)"
    SimSpaceControlData="(LinearStrength=0,LinearDampingRatio=1,LinearExtraDamping=0,MaxForce=0,AngularStrength=0,AngularDampingRatio=1,AngularExtraDamping=0,MaxTorque=0,LinearTargetVelocityMultiplier=1,AngularTargetVelocityMultiplier=1,SkeletalAnimationVelocityMultiplier=1,Custom ...(truncated)"
    Solver=(PhysicsSolverComponentKey=(ElementKey=(Type=Bone,Name="None"),Name="PhysicsBody"),bUseAutomaticSolver=false,SourceBone=(Type=None,Name="None"),TargetBone=(Type=None,Name="None"),bIncludeInChecksForReset=True)
    bAddJoint=True
    bAddParentSpaceControl=false
    bAddSimSpaceControl=true
    Dynamics.Density=@"Curve Gain".Result
    Dynamics.AngularDamping=@"Curve Gain_1".Result
    BodyData.GravityMultiplier=@"Curve Gain_2".Result
    JointData.ConeConstraint.Swing1LimitDegrees=@"Curve Gain_4".Result
    JointData.ConeConstraint.Swing2LimitDegrees=@"Curve Gain_5".Result
    JointData.TwistConstraint.TwistLimitDegrees=@"Curve Gain_6".Result
    SimSpaceControlData.AngularStrength=@"Curve Gain_7".Result
    Dynamics.MassOverride=@Divide.Result
    Owner=@For_Each.Element
    JointData.ExtraParentOffset.Rotation=@Interpolate.Result
    Dynamics.CentreOfMassOverride=/Script/RigVM.RigVMFunction_MathVectorLerp(T=@"Curve Gain_3".Result, A=@RigVMFunction_MathVectorMake.Result, B=RigVMVariableNode() {variable=Centre Of Mass Override}.Value) {is=struct RigVMUnitNode, template="Interpolate::Execute(in A,in B,in T,out Result)"}.Result
    Dynamics.bOverrideCentreOfMass=RigVMVariableNode() {variable=Override Centre Of Mass}.Value
    Solver.PhysicsSolverComponentKey=RigVMVariableNode() {variable=Solver}.Value
   ) {is=struct RigVMUnitNode}
   > RigVMVariableNode(Value=@Divide.Result) {variable=Current Mass}
   > /Script/ControlRigPhysics.RigUnit_HierarchyDisableCollisionBetween(PhysicsBodyComponentKey1=@n2.PhysicsBodyComponentKey, PhysicsBodyComponentKey2=RigVMVariableNode() {variable=collision body}.Value) {is=struct RigVMUnitNode}
   > RigVMVariableNode(Value=@n2.PhysicsBodyComponentKey) {variable=collision body}
   > &RigUnit_HierarchyAutoCalculateCollision(
    MinSize=0.01
    PhysicsBodyComponentKey=@n2.PhysicsBodyComponentKey
    MinAspectRatio=&RigVMFunction_AnimEvalRichCurve(
     SourceMinimum=0
     TargetMinimum=0
     Curve=RigVMVariableNode() {variable=Thickness Ramp}.Value
     Value=RigVMVariableNode() {variable=Chain Ratio}.Value
    ).Result
   )
   > Add = RigVMDispatchNode(Element=@n2.PhysicsBodyComponentKey, Array=RigVMVariableNode() {variable=Body Components}.Value) {template="DISPATCH_RigVMDispatch_ArrayAdd(io Array,in Element,out Index)"}
   > RigVMDispatchNode(Element=@n2.SimSpaceControlComponentKey, Array=RigVMVariableNode() {variable=Sim Space Components}.Value) {template="DISPATCH_RigVMDispatch_ArrayAdd(io Array,in Element,out Index)"}
   > RigVMDispatchNode(Element=@n2.PhysicsJointComponentKey, Array=RigVMVariableNode() {variable=Joint Components}.Value) {template="DISPATCH_RigVMDispatch_ArrayAdd(io Array,in Element,out Index)"}
   RigVMFunction_Sequence_1 = RigVMAggregateNode(ExecuteContext=@RigVMFunction_Sequence.E)
   For_Each_1 = RigVMDispatchNode(ExecuteContext=@RigVMFunction_Sequence_1.A, Array=RigVMVariableNode() {variable=Body Components}.Value) {template="DISPATCH_RigVMDispatch_ArrayIterator(in Array,out Element,out Index,out Count,out Ratio)"}
   > Set_Metadata = RigVMDispatchNode(Name=Has Physics Body, NameSpace=None, Item=@For_Each_1.Element.ElementKey, Value=@For_Each_1.Element.Name) {template="DISPATCH_RigDispatch_SetMetadata(in Item,in Name,in NameSpace,in Value,out Success)"}
   For_Each_3 = RigVMDispatchNode(ExecuteContext=@RigVMFunction_Sequence_1.B, Array=RigVMVariableNode() {variable=Joint Components}.Value) {template="DISPATCH_RigVMDispatch_ArrayIterator(in Array,out Element,out Index,out Count,out Ratio)"}
   > Set_Metadata_2 = RigVMDispatchNode(Name=Has Physics Joint, NameSpace=None, Item=@For_Each_3.Element.ElementKey, Value=@For_Each_3.Element.Name) {template="DISPATCH_RigDispatch_SetMetadata(in Item,in Name,in NameSpace,in Value,out Success)"}
   For_Each_2 = RigVMDispatchNode(ExecuteContext=@RigVMFunction_Sequence_1.C, Array=RigVMVariableNode() {variable=Sim Space Components}.Value) {template="DISPATCH_RigVMDispatch_ArrayIterator(in Array,out Element,out Index,out Count,out Ratio)"}
   > Set_Metadata_3 = RigVMDispatchNode(Name=Has Sim Space, NameSpace=None, Item=@For_Each_2.Element.ElementKey, Value=@For_Each_2.Element.Name) {template="DISPATCH_RigDispatch_SetMetadata(in Item,in Name,in NameSpace,in Value,out Success)"}
   n3 = &RigVMFunction_ControlFlowBranch(
    Condition=/Script/ControlRig.RigUnit_ResolveConnector(Connector=(Type=Connector,Name="End"), SkipSocket=False) {is=struct RigVMUnitNode}.bIsConnected
    ExecuteContext=@RigVMFunction_Sequence.A
   )
   RigVMVariableNode(
    Value=/Script/ControlRig.RigUnit_HierarchyGetParent(Child=(Type=Connector,Name="Start"), bDefaultParent=True) {is=struct RigVMUnitNode}.Parent
    ExecuteContext=@n3.Completed
   ) {variable=Kinematic Parent}
   RigVMVariableNode(
    Value=((Type=None,Name="None"))
    Value.0=&RigUnit_ResolveConnector(Connector=(Type=Connector,Name="Start"), SkipSocket=False).Result
    ExecuteContext=@n3.False
   ) {variable=Chain}
   n4 = /Script/ControlRig.RigUnit_HierarchyGetChainItemArray(End=(Type=Connector,Name="End"), Start=(Type=Connector,Name="Start"), bIncludeEnd=True, bIncludeStart=True, bReverse=False) {is=struct RigVMUnitNode}
   RigVMVariableNode(Value=@n4.Chain, ExecuteContext=@n3.True) {variable=Chain}
   Get_Module_Metadata = RigVMDispatchNode(Default=(Type=Bone,Name="None"), Name=Physics Solver Owner, NameSpace=Parent) {template="DISPATCH_RigDispatch_GetModuleMetadata(in Name,in NameSpace,in Default,out Value,out Found)"}
   Get_Module_Metadata_1 = RigVMDispatchNode(Default=PhysicsSolver, Name=Physics Solver Name, NameSpace=Parent) {template="DISPATCH_RigDispatch_GetModuleMetadata(in Name,in NameSpace,in Default,out Value,out Found)"}
   RigVMVariableNode(
    Value=(ElementKey=(Type=None,Name="None"),Name="None")
    Value.ElementKey=@Get_Module_Metadata.Value
    Value.Name=@Get_Module_Metadata_1.Value
    ExecuteContext=@RigVMFunction_Sequence.B
   ) {variable=Solver}
   RigVMFunction_MathBoolAnd = &RigVMFunction_MathBoolAnd(A=@Get_Module_Metadata.Found, B=@Get_Module_Metadata_1.Found)
   > RigVMVariableNode(Value=@RigVMFunction_MathBoolAnd.Result) {variable=Solver Exists}
   > Set_Module_Metadata = RigVMDispatchNode(Name=Physics Solver Owner, NameSpace=Self, Value=@Get_Module_Metadata.Value) {template="DISPATCH_RigDispatch_SetModuleMetadata(in Name,in NameSpace,in Value,out Success)"}
   > Set_Module_Metadata_1 = RigVMDispatchNode(Name=Physics Solver Name, NameSpace=Self, Value=@Get_Module_Metadata_1.Value) {template="DISPATCH_RigDispatch_SetModuleMetadata(in Name,in NameSpace,in Value,out Success)"}
  ]}
 New Function_ContainedGraph {nodes [
   n1 = RigVMFunctionEntryNode()
   n2 = /Script/RigVM.RigVMFunction_TimeLoop(Normalize=true, Speed=RigVMVariableNode() {variable=Speed Y}.Value, Duration=RigVMVariableNode() {variable=Duration Y}.Value) {is=struct RigVMUnitNode}
   Noise = /Script/RigVM.RigVMFunction_NoiseDouble(
    Frequency=2
    Speed=0.015
    Value=@n2.Relative
    Minimum=RigVMVariableNode() {variable=Min Y}.Value
    Maximum=RigVMVariableNode() {variable=Max Y}.Value
   ) {is=struct RigVMUnitNode, template="Noise::Execute(in Value,in Speed,in Frequency,in Minimum,in Maximum,out Result)"}
   KalmanFilter = &RigVMFunction_KalmanFloat(BufferSize=16, Value=@Noise.Result) {template="KalmanFilter::Execute(in Value,in BufferSize,out Result)"}
   n3 = /Script/RigVM.RigVMFunction_AnimRichCurve(Curve=(EditorCurveData=(Keys=((InterpMode=RCIM_Cubic,TangentMode=RCTM_User),(InterpMode=RCIM_Cubic,TangentMode=RCTM_User,Time=0.5,Value=1),(InterpMode=RCIM_Cubic,TangentMode=RCTM_User,Time=1))))) {is=struct RigVMUnitNode}
   Interpolate = /Script/RigVM.RigVMFunction_MathDoubleLerp(
    B=@KalmanFilter.Result
    T=&RigVMFunction_NoiseDouble(Frequency=0.1, Maximum=1, Minimum=0, Speed=1, Value=@n2.Relative) {template="Noise::Execute(in Value,in Speed,in Frequency,in Minimum,in Maximum,out Result)"}.Result
    A=&RigVMFunction_AnimEvalRichCurve(
     SourceMinimum=0
     Curve=@n3.Curve
     Value=@n2.Relative
     TargetMinimum=RigVMVariableNode() {variable=Min Y}.Value
     TargetMaximum=RigVMVariableNode() {variable=Max Y}.Value
    ).Result
   ) {is=struct RigVMUnitNode, template="Interpolate::Execute(in A,in B,in T,out Result)"}
   n4 = &RigVMFunction_MathQuaternionToEuler(RotationOrder=ZYX, Value=&RigUnit_GetTransform(bInitial=true, Item=@n1.Item).Transform.Rotation)
   n5 = &RigVMFunction_MathDoubleAdd(B=@Interpolate.Result, A=@n4.Result.Y)
   n6 = RigVMVariableNode() {variable=Alpha}
   n7 = &RigVMFunction_TimeLoop(Normalize=true, Speed=RigVMVariableNode() {variable=Speed Z}.Value, Duration=RigVMVariableNode() {variable=Duration Z}.Value)
   n8 = &RigVMFunction_NoiseDouble(
    Frequency=1
    Speed=0.1
    Value=@n7.Relative
    Maximum=RigVMVariableNode() {variable=Max Z}.Value
    Minimum=RigVMVariableNode() {variable=Min Z}.Value
   ) {template="Noise::Execute(in Value,in Speed,in Frequency,in Minimum,in Maximum,out Result)"}
   n9 = RigVMVariableNode() {variable=Max Z}
   n10 = RigVMVariableNode() {variable=Min Z}
   n11 = &RigVMFunction_MathDoubleAdd(
    B=&RigVMFunction_MathDoubleLerp(
     B=&RigVMFunction_KalmanFloat(BufferSize=16, Value=@n8.Result) {template="KalmanFilter::Execute(in Value,in BufferSize,out Result)"}.Result
     T=&RigVMFunction_NoiseDouble(Frequency=0.2, Maximum=1, Minimum=0, Speed=1.1, Value=@n7.Relative) {template="Noise::Execute(in Value,in Speed,in Frequency,in Minimum,in Maximum,out Result)"}.Result
     A=&RigVMFunction_AnimEvalRichCurve(SourceMinimum=0, Curve=@n3.Curve, Value=@n7.Relative, TargetMaximum=@n9.Value, TargetMinimum=@n10.Value).Result
    ) {template="Interpolate::Execute(in A,in B,in T,out Result)"}.Result
    A=@n4.Result.Z
   )
   n12 = RigVMVariableNode() {variable=Alpha}
   > RigVMFunctionReturnNode(Rotation=&RigVMFunction_MathQuaternionFromEuler(
     Euler=(X=0,Y=0,Z=0)
     Euler.Y=&RigVMFunction_MathDoubleLerp(B=@n5.Result, A=@n4.Result.Y, T=@n6.Value) {template="Interpolate::Execute(in A,in B,in T,out Result)"}.Result
     Euler.Z=&RigVMFunction_MathDoubleLerp(B=@n11.Result, A=@n4.Result.Z, T=@n12.Value) {template="Interpolate::Execute(in A,in B,in T,out Result)"}.Result
     Euler.X=@n4.Result.X
    ).Result)
  ]}
 New Function_ContainedGraph {nodes [
   n1 = RigVMFunctionEntryNode()
   RigVMFunction_AnimEvalRichCurve = &RigVMFunction_AnimEvalRichCurve(SourceMinimum=0, TargetMinimum=0, Curve=@n1.Curve, Value=@n1.Ratio)
   Multiply = &RigVMFunction_MathDoubleMul(A=@n1.Value, B=@RigVMFunction_AnimEvalRichCurve.Result)
   > RigVMFunctionReturnNode(Result=@Multiply.Result)
  ]}
 RigVMFunctionLibrary {nodes [
   Curve Gain = RigVMCollapseNode(Curve=())
   "Keep Alive " = RigVMCollapseNode(Alpha=1, Duration Y=2, Duration Z=1, Item=(Type=None,Name="None"), Max Y=10, Max Z=20, Min Y=-10, Min Z=5, Speed Y=0.25, Speed Z=0.25)
  ]}
 RigVMFunction_Sequence_1_ContainedGraph {nodes [
   n1 = RigVMFunctionEntryNode()
   RigVMFunction_Sequence = &RigVMFunction_Sequence(ExecuteContext=@n1.ExecuteContext)
   n2 = &RigVMFunction_Sequence(ExecuteContext=@RigVMFunction_Sequence.B)
   n3 = &RigVMFunction_Sequence(ExecuteContext=@n2.B)
   n4 = &RigVMFunction_Sequence(ExecuteContext=@n3.B)
   RigVMFunctionReturnNode(A=@RigVMFunction_Sequence.A, B=@n2.A, C=@n3.A, D=@n4.A, E=@n4.B)
  ]}
 RigVMFunction_Sequence_2_ContainedGraph {nodes [
   n1 = RigVMFunctionEntryNode()
   n2 = &RigVMFunction_Sequence(ExecuteContext=@n1.ExecuteContext)
   n3 = &RigVMFunction_Sequence(ExecuteContext=@n2.B)
   RigVMFunctionReturnNode(A=@n2.A, B=@n3.A, C=@n3.B)
  ]}
 RigVMModel {nodes [
   RigUnit_BeginExecution = &RigUnit_BeginExecution()
   RigVMFunction_Sequence = &RigVMFunction_Sequence(ExecuteContext=@RigUnit_BeginExecution.ExecutePin)
   For_Each_2 = RigVMDispatchNode(ExecuteContext=@RigVMFunction_Sequence.A, Array=RigVMVariableNode() {variable=Chain}.Value) {template="DISPATCH_RigVMDispatch_ArrayIterator(in Array,out Element,out Index,out Count,out Ratio)"}
   RigVMFunction_AnimEvalRichCurve = &RigVMFunction_AnimEvalRichCurve(SourceMinimum=0, TargetMinimum=0, Value=@For_Each_2.Ratio, Curve=RigVMVariableNode() {variable=Rotation Offset Ramp}.Value)
   > &RigUnit_ModifyTransforms(
    ItemToModify=((Item=(Type=Bone,Name="None"),Transform=(Rotation=(X=0,Y=0,Z=0,W=1),Translation=(X=0,Y=0,Z=0),Scale3D=(X=1,Y=1,Z=1))))
    Weight=1
    WeightMinimum=0
    ItemToModify.0.Item=@For_Each_2.Element
    ItemToModify.0.Transform.Rotation=&RigVMFunction_MathQuaternionSlerp(A=(X=0,Y=0,Z=0,W=1), T=@RigVMFunction_AnimEvalRichCurve.Result, B=RigVMVariableNode() {variable=Rotation Offset}.Value) {template="Interpolate::Execute(in A,in B,in T,out Result)"}.Result
   )
   /Script/ControlRigPhysics.RigUnit_HierarchySetControlCustomControlPoint(
    CustomControlPoint=(X=0,Y=0,Z=0)
    ExecutePin=()
    PhysicsControlComponentKey=(ElementKey=(Type=Bone,Name="None"),Name="PhysicsControl")
    bUseCustomControlPoint=True
   ) {is=struct RigVMUnitNode}
   /Script/ControlRigPhysics.RigUnit_HierarchySetControlData(
    ControlData="(bEnabled=True,LinearStrength=0,LinearDampingRatio=1,LinearExtraDamping=0,MaxForce=0,AngularStrength=0,AngularDampingRatio=1,AngularExtraDamping=0,MaxTorque=0,LinearTargetVelocityMultiplier=1,AngularTargetVelocityMultiplier=1,CustomControlPoint=(X=0,Y=0,Z= ...(truncated)"
    ExecutePin=()
    PhysicsControlComponentKey=(ElementKey=(Type=Bone,Name="None"),Name="PhysicsControl")
   ) {is=struct RigVMUnitNode}
   /Script/ControlRigPhysics.RigUnit_HierarchySetControlEnabled(ExecutePin=(), PhysicsControlComponentKey=(ElementKey=(Type=Bone,Name="None"),Name="PhysicsControl"), bEnabled=True) {is=struct RigVMUnitNode}
   /Script/ControlRigPhysics.RigUnit_HierarchySetControlMultiplier(
    ControlMultiplier="(LinearStrengthMultiplier=(X=1,Y=1,Z=1),LinearDampingRatioMultiplier=(X=1,Y=1,Z=1),LinearExtraDampingMultiplier=(X=1,Y=1,Z=1),MaxForceMultiplier=(X=1,Y=1,Z=1),AngularStrengthMultiplier=1,AngularDampingRatioMultiplier=1,AngularExtraDampingMultiplier=1,MaxTo ...(truncated)"
    ExecutePin=()
    PhysicsControlComponentKey=(ElementKey=(Type=Bone,Name="None"),Name="PhysicsControl")
   ) {is=struct RigVMUnitNode}
   /Script/ControlRigPhysics.RigUnit_HierarchySetControlTarget(
    ControlTarget=(TargetPosition=(X=0,Y=0,Z=0),TargetVelocity=(X=0,Y=0,Z=0),TargetOrientation=(Pitch=0,Yaw=0,Roll=0),TargetAngularVelocity=(X=0,Y=0,Z=0),bApplyControlPointToTarget=False)
    ExecutePin=()
    PhysicsControlComponentKey=(ElementKey=(Type=Bone,Name="None"),Name="PhysicsControl")
   ) {is=struct RigVMUnitNode}
   /Script/ControlRigPhysics.RigUnit_HierarchySetDynamics(
    Dynamics="(Density=1,MassOverride=1,bOverrideCentreOfMass=False,CentreOfMassOverride=(X=0,Y=0,Z=0),CentreOfMassNudge=(X=0,Y=0,Z=0),bOverrideMomentsOfInertia=False,MomentsOfInertiaOverride=(X=1,Y=1,Z=1),LinearDamping=0,AngularDamping=0,bScaleDampingByInverseMass=Fals ...(truncated)"
    ExecutePin=()
    PhysicsBodyComponentKey=(ElementKey=(Type=Bone,Name="None"),Name="PhysicsBody")
   ) {is=struct RigVMUnitNode}
   /Script/ControlRigPhysics.RigUnit_HierarchySetJointData(
    ExecutePin=()
    JointData="(bEnabled=True,bAutoCalculateParentOffset=True,ExtraParentOffset=(Rotation=(X=0,Y=0,Z=0,W=1),Translation=(X=0,Y=0,Z=0),Scale3D=(X=1,Y=1,Z=1)),bAutoCalculateChildOffset=True,ExtraChildOffset=(Rotation=(X=0,Y=0,Z=0,W=1),Translation=(X=0,Y=0,Z=0),Scale3D=(X=1 ...(truncated)"
    PhysicsJointComponentKey=(ElementKey=(Type=Bone,Name="None"),Name="PhysicsJoint")
   ) {is=struct RigVMUnitNode}
   /Script/ControlRigPhysics.RigUnit_HierarchySetJointDriveData(
    DriveData="(LinearDriveConstraint=(PositionTarget=(X=0,Y=0,Z=0),VelocityTarget=(X=0,Y=0,Z=0),XDrive=(Stiffness=50,Damping=1,MaxForce=0,bEnablePositionDrive=False,bEnableVelocityDrive=False),YDrive=(Stiffness=50,Damping=1,MaxForce=0,bEnablePositionDrive=False,bEnableV ...(truncated)"
    ExecutePin=()
    PhysicsJointComponentKey=(ElementKey=(Type=Bone,Name="None"),Name="PhysicsJoint")
   ) {is=struct RigVMUnitNode}
   /Script/ControlRigPhysics.RigUnit_HierarchySetJointDriveUseSkeletalAnimation(ExecutePin=(), PhysicsJointComponentKey=(ElementKey=(Type=Bone,Name="None"),Name="PhysicsJoint"), bUseSkeletalAnimation=True) {is=struct RigVMUnitNode}
   /Script/ControlRigPhysics.RigUnit_HierarchySetPhysicsBodyCollisionType(CollisionType=QueryAndPhysics, ExecutePin=(), PhysicsBodyComponentKey=(ElementKey=(Type=Bone,Name="None"),Name="PhysicsBody")) {is=struct RigVMUnitNode}
   /Script/ControlRigPhysics.RigUnit_HierarchySetPhysicsBodyDamping(AngularDamping=0, ExecutePin=(), LinearDamping=0, PhysicsBodyComponentKey=(ElementKey=(Type=Bone,Name="None"),Name="PhysicsBody")) {is=struct RigVMUnitNode}
   /Script/ControlRigPhysics.RigUnit_HierarchySetPhysicsBodyGravityMultiplier(ExecutePin=(), GravityMultiplier=1, PhysicsBodyComponentKey=(ElementKey=(Type=Bone,Name="None"),Name="PhysicsBody")) {is=struct RigVMUnitNode}
   /Script/ControlRigPhysics.RigUnit_HierarchySetPhysicsBodyKinematicTarget(
    ExecutePin=()
    KinematicTarget=(Rotation=(X=0,Y=0,Z=0,W=1),Translation=(X=0,Y=0,Z=0),Scale3D=(X=1,Y=1,Z=1))
    PhysicsBodyComponentKey=(ElementKey=(Type=Bone,Name="None"),Name="PhysicsBody")
   ) {is=struct RigVMUnitNode}
   /Script/ControlRigPhysics.RigUnit_HierarchySetPhysicsBodyMovementType(ExecutePin=(), MovementType=Simulated, PhysicsBodyComponentKey=(ElementKey=(Type=Bone,Name="None"),Name="PhysicsBody")) {is=struct RigVMUnitNode}
   /Script/ControlRigPhysics.RigUnit_HierarchySetPhysicsBodyPhysicsBlendWeight(ExecutePin=(), PhysicsBlendWeight=1, PhysicsBodyComponentKey=(ElementKey=(Type=Bone,Name="None"),Name="PhysicsBody")) {is=struct RigVMUnitNode}
   /Script/ControlRigPhysics.RigUnit_HierarchySetPhysicsBodySparseData(
    Data="(MovementType=Simulated,CollisionType=QueryAndPhysics,GravityMultiplier=1,PhysicsBlendWeight=1,KinematicTargetSpace=OffsetInBoneSpace,bUpdateKinematicFromSimulation=True,bEnableCCD=False,bEnableMovementType=True,bEnableCollisionType=True,bEnableGravityMult ...(truncated)"
    ExecutePin=()
    PhysicsBodyComponentKey=(ElementKey=(Type=Bone,Name="None"),Name="PhysicsBody")
   ) {is=struct RigVMUnitNode}
   /Script/ControlRigPhysics.RigUnit_HierarchySetPhysicsBodyTargetBone(ExecutePin=(), PhysicsBodyComponentKey=(ElementKey=(Type=Bone,Name="None"),Name="PhysicsBody"), TargetBone=(Type=Bone,Name="None")) {is=struct RigVMUnitNode}
  ]}
]
hierarchy [
 Start {type=Connector, loc=<0 0 0>, rot=<0 0 0 1>}
 Root {type=Socket, loc=<0 0 0>, rot=<0 0 0 1>}
 End {type=Connector, loc=<0 0 0>, rot=<0 0 0 1>}
]
variables [
 Angular Damping {type=double, default=0}
 Angular Damping Ramp {
  type=FRuntimeFloatCurve
  default=(EditorCurveData=(Keys=((Value=1),(Time=0.5,Value=1),(Time=1,Value=1)),DefaultValue=340282346638528859811704183484516925440,PreInfinityExtrap=RCCE_Constant,PostInfinityExtrap=RCCE_Constant),ExternalCurve=None)
 }
 Angular Strength {type=double, default=0}
 Angular Strength Ramp {
  type=FRuntimeFloatCurve
  default=(EditorCurveData=(Keys=((Value=1),(Time=0.5,Value=1),(Time=1,Value=1)),DefaultValue=340282346638528859811704183484516925440,PreInfinityExtrap=RCCE_Constant,PostInfinityExtrap=RCCE_Constant),ExternalCurve=None)
 }
 Body Components {type="TArray<FRigComponentKey>"}
 Centre Of Mass Override {type=FVector, default=(X=0,Y=0,Z=0)}
 Centre Of Mass Override Ramp {
  type=FRuntimeFloatCurve
  default=(EditorCurveData=(Keys=((Value=1),(Time=0.5,Value=1),(Time=1,Value=1)),DefaultValue=340282346638528859811704183484516925440,PreInfinityExtrap=RCCE_Constant,PostInfinityExtrap=RCCE_Constant),ExternalCurve=None)
 }
 Chain {type="TArray<FRigElementKey>"}
 Chain Ratio {type=double, default=0}
 Colliders {type=bool, default=False}
 Cone Rotation Offset {type=FQuat, default=(X=0,Y=0,Z=0,W=1)}
 Cone Rotation Offset Ramp {
  type=FRuntimeFloatCurve
  default=(EditorCurveData=(Keys=((Value=1),(Time=0.5,Value=1),(Time=1,Value=1)),DefaultValue=340282346638528859811704183484516925440,PreInfinityExtrap=RCCE_Constant,PostInfinityExtrap=RCCE_Constant),ExternalCurve=None)
 }
 Current Mass {type=double, default=0}
 Density {type=double, default=1}
 Density Ramp {
  type=FRuntimeFloatCurve
  default=(EditorCurveData=(Keys=((Value=1),(Time=0.5,Value=1),(Time=1,Value=1)),DefaultValue=340282346638528859811704183484516925440,PreInfinityExtrap=RCCE_Constant,PostInfinityExtrap=RCCE_Constant),ExternalCurve=None)
 }
 Enable Physics {type=bool, default=True}
 Gravity Multiplier {type=double, default=1}
 Gravity Ramp {
  type=FRuntimeFloatCurve
  default=(EditorCurveData=(Keys=((Value=1),(Time=0.5,Value=1),(Time=1,Value=1)),DefaultValue=340282346638528859811704183484516925440,PreInfinityExtrap=RCCE_Constant,PostInfinityExtrap=RCCE_Constant),ExternalCurve=None)
 }
 Ground Bone {type=FRigElementKey, default=(Type=None,Name="")}
 Joint Components {type="TArray<FRigComponentKey>"}
 Kinematic Parent {type=FRigElementKey, default=(Type=None,Name="")}
 Override Centre Of Mass {type=bool, default=False}
 Owner {type=FRigElementKey, default=(Type=None,Name="")}
 Rotation Offset {type=FQuat, default=(X=0,Y=0,Z=0,W=1)}
 Rotation Offset Ramp {
  type=FRuntimeFloatCurve
  default=(EditorCurveData=(Keys=((Value=1),(Time=1,Value=1)),DefaultValue=340282346638528859811704183484516925440,PreInfinityExtrap=RCCE_Constant,PostInfinityExtrap=RCCE_Constant),ExternalCurve=None)
 }
 Sim Space Components {type="TArray<FRigComponentKey>"}
 Solver {type=FRigComponentKey, default=(ElementKey=(Type=None,Name=""),Name="")}
 Solver Exists {type=bool, default=False}
 Start Mass {type=double, default=100}
 Swing 1 Limit {type=double, default=45}
 Swing 1 Limit Ramp {
  type=FRuntimeFloatCurve
  default=(EditorCurveData=(Keys=((Value=1),(Time=0.5,Value=1),(Time=1,Value=1)),DefaultValue=340282346638528859811704183484516925440,PreInfinityExtrap=RCCE_Constant,PostInfinityExtrap=RCCE_Constant),ExternalCurve=None)
 }
 Swing 2 Limit {type=double, default=45}
 Swing 2 Limit Ramp {
  type=FRuntimeFloatCurve
  default=(EditorCurveData=(Keys=((Value=1),(Time=0.5,Value=1),(Time=1,Value=1)),DefaultValue=340282346638528859811704183484516925440,PreInfinityExtrap=RCCE_Constant,PostInfinityExtrap=RCCE_Constant),ExternalCurve=None)
 }
 Thickness Ramp {
  type=FRuntimeFloatCurve
  default=(EditorCurveData=(Keys=((Value=0.5),(Time=0.5,Value=0.5),(Time=1,Value=0.5)),DefaultValue=340282346638528859811704183484516925440,PreInfinityExtrap=RCCE_Constant,PostInfinityExtrap=RCCE_Constant),ExternalCurve=None)
 }
 Twist Limit {type=double, default=8}
 Twist Limit Ramp {
  type=FRuntimeFloatCurve
  default=(EditorCurveData=(Keys=((Value=1),(Time=0.5,Value=1),(Time=1,Value=1)),DefaultValue=340282346638528859811704183484516925440,PreInfinityExtrap=RCCE_Constant,PostInfinityExtrap=RCCE_Constant),ExternalCurve=None)
 }
 collision body {type=FRigComponentKey, default=(ElementKey=(Type=None,Name=""),Name="")}
]
