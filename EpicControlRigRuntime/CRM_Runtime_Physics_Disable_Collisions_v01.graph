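# Control Rig module that disables physics collisions between selected bodies during the prepare-for-execution event.
# It gathers elements tagged with "Has Physics Body" metadata, whose value names each physics body component, and matches bones from two array connectors against them.
# Remove All Collisions pairs each A bone with every physics owner; otherwise A bones pair only with B bones.
# An excluded-owners branch is unreachable behind a constant-true condition and unfinished.
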ControlRigBlueprint /EpicControlRigRuntime/CRM_Runtime_Physics_Disable_Collisions_v01
graphs [
 RigVMFunctionLibrary {links [], nodes []}
 RigVMFunction_Sequence_2_ContainedGraph {nodes [
   n1 = RigVMFunctionEntryNode()
   n2 = &RigVMFunction_Sequence(ExecuteContext=@n1.ExecuteContext)
   n3 = &RigVMFunction_Sequence(ExecuteContext=@n2.B)
   RigVMFunctionReturnNode(A=@n2.A, B=@n3.A, C=@n3.B)
  ]}
 RigVMModel {nodes [
   RigUnit_PrepareForExecution = &RigUnit_PrepareForExecution()
   RigVMFunction_Sequence = &RigVMFunction_Sequence(ExecuteContext=@RigUnit_PrepareForExecution.ExecutePin)
   n1 = RigVMVariableNode() {variable=Remove All Collisions}
   RigVMFunction_ControlFlowBranch = &RigVMFunction_ControlFlowBranch(ExecuteContext=@RigVMFunction_Sequence.B, Condition=@n1.Value)
   n2 = &RigVMFunction_ControlFlowBranch(Condition=true, ExecuteContext=@RigVMFunction_ControlFlowBranch.False)
   Clone = RigVMDispatchNode(Array=RigVMVariableNode() {variable=Physics Owners}.Value) {template="DISPATCH_RigVMDispatch_ArrayClone(in Array,out Clone)"}
   RigVMVariableNode(Value=@Clone.Clone, ExecuteContext=@n2.False) {variable=Excluded Physics Owners}
   > RigVMFunction_Sequence_1 = RigVMAggregateNode()
   n3 = RigVMDispatchNode(
    Array=/Script/ControlRig.RigUnit_ResolveArrayConnector(Connector=(Type=Connector,Name="Disable Bones A"), SkipSocket=False) {is=struct RigVMUnitNode}.Result
    Other=&RigUnit_ResolveArrayConnector(Connector=(Type=Connector,Name="Disable Bones B"), SkipSocket=False).Result
    ExecuteContext=@RigVMFunction_Sequence_1.A
   ) {template="DISPATCH_RigVMDispatch_ArrayUnion(io Array,in Other)"}
   > For_Each_4 = RigVMDispatchNode(Array=@n3.Array) {template="DISPATCH_RigVMDispatch_ArrayIterator(in Array,out Element,out Index,out Count,out Ratio)"}
   n4 = RigVMDispatchNode(Element=@For_Each_4.Element, Array=RigVMVariableNode() {variable=Physics Owners}.Value) {template="DISPATCH_RigVMDispatch_ArrayFind(in Array,in Element,out Index,out Success)"}
   n5 = &RigVMFunction_ControlFlowBranch(Condition=@n4.Success, ExecuteContext=@For_Each_4.ExecuteContext)
   Remove = RigVMDispatchNode(ExecuteContext=@n5.True, Array=RigVMVariableNode() {variable=Excluded Physics Owners}.Value, Index=@n4.Index) {template="DISPATCH_RigVMDispatch_ArrayRemove(io Array,in Index)"}
   For_Each_6 = RigVMDispatchNode(ExecuteContext=@RigVMFunction_Sequence_1.C, Array=RigVMVariableNode() {variable=Excluded Physics Owners}.Value) {template="DISPATCH_RigVMDispatch_ArrayIterator(in Array,out Element,out Index,out Count,out Ratio)"}
   For_Each = RigVMDispatchNode(
    Array=&RigUnit_ResolveArrayConnector(Connector=(Type=Connector,Name="Disable Bones A"), SkipSocket=False).Result
    ExecuteContext=@n2.True
   ) {template="DISPATCH_RigVMDispatch_ArrayIterator(in Array,out Element,out Index,out Count,out Ratio)"}
   n6 = RigVMDispatchNode(Element=@For_Each.Element, Array=RigVMVariableNode() {variable=Physics Owners}.Value) {template="DISPATCH_RigVMDispatch_ArrayFind(in Array,in Element,out Index,out Success)"}
   > n7 = RigVMVariableNode(Value=@n6.Index) {variable=Index A}
   n8 = &RigVMFunction_ControlFlowBranch(Condition=@n6.Success, ExecuteContext=@n7.ExecuteContext)
   For_Each_2 = RigVMDispatchNode(
    Array=&RigUnit_ResolveArrayConnector(Connector=(Type=Connector,Name="Disable Bones B"), SkipSocket=False).Result
    ExecuteContext=@n8.True
   ) {template="DISPATCH_RigVMDispatch_ArrayIterator(in Array,out Element,out Index,out Count,out Ratio)"}
   n9 = RigVMDispatchNode(Element=@For_Each_2.Element, Array=RigVMVariableNode() {variable=Physics Owners}.Value) {template="DISPATCH_RigVMDispatch_ArrayFind(in Array,in Element,out Index,out Success)"}
   > n10 = RigVMVariableNode(Value=@n9.Index) {variable=Index B}
   n11 = &RigVMFunction_ControlFlowBranch(Condition=@n9.Success, ExecuteContext=@n10.ExecuteContext)
   n12 = RigVMDispatchNode(Array=RigVMVariableNode() {variable=Physics Owners}.Value, Index=RigVMVariableNode() {variable=Index A}.Value) {template="DISPATCH_RigVMDispatch_ArrayGetAtIndex(in Array,in Index,out Element)"}
   n13 = RigVMDispatchNode(Array=RigVMVariableNode() {variable=Physics Owners}.Value, Index=RigVMVariableNode() {variable=Index B}.Value) {template="DISPATCH_RigVMDispatch_ArrayGetAtIndex(in Array,in Index,out Element)"}
   Get_Metadata_1 = RigVMDispatchNode(Default=PhysicsBody, Name=Has Physics Body, NameSpace=None, Item=@n13.Element) {template="DISPATCH_RigDispatch_GetMetadata(in Item,in Name,in NameSpace,in Default,out Value,out Found)"}
   Get_Metadata_2 = RigVMDispatchNode(Default=PhysicsBody, Name=Has Physics Body, NameSpace=None, Item=@n12.Element) {template="DISPATCH_RigDispatch_GetMetadata(in Item,in Name,in NameSpace,in Default,out Value,out Found)"}
   /Script/ControlRigPhysics.RigUnit_HierarchyDisableCollisionBetween(
    PhysicsBodyComponentKey1=(ElementKey=(Type=Bone,Name="None"),Name="PhysicsBody")
    PhysicsBodyComponentKey2=(ElementKey=(Type=Bone,Name="None"),Name="PhysicsBody")
    PhysicsBodyComponentKey1.ElementKey=@n12.Element
    PhysicsBodyComponentKey2.ElementKey=@n13.Element
    PhysicsBodyComponentKey2.Name=@Get_Metadata_1.Value
    PhysicsBodyComponentKey1.Name=@Get_Metadata_2.Value
    ExecutePin=@n11.True
   ) {is=struct RigVMUnitNode}
   For_Each_1 = RigVMDispatchNode(
    Array=&RigUnit_ResolveArrayConnector(Connector=(Type=Connector,Name="Disable Bones A"), SkipSocket=False).Result
    ExecuteContext=@RigVMFunction_ControlFlowBranch.True
   ) {template="DISPATCH_RigVMDispatch_ArrayIterator(in Array,out Element,out Index,out Count,out Ratio)"}
   n14 = RigVMDispatchNode(Element=@For_Each_1.Element, Array=RigVMVariableNode() {variable=Physics Owners}.Value) {template="DISPATCH_RigVMDispatch_ArrayFind(in Array,in Element,out Index,out Success)"}
   > n15 = RigVMVariableNode(Value=@n14.Index) {variable=Index A}
   n16 = &RigVMFunction_ControlFlowBranch(Condition=@n14.Success, ExecuteContext=@n15.ExecuteContext)
   For_Each_3 = RigVMDispatchNode(ExecuteContext=@n16.True, Array=RigVMVariableNode() {variable=Physics Owners}.Value) {template="DISPATCH_RigVMDispatch_ArrayIterator(in Array,out Element,out Index,out Count,out Ratio)"}
   n17 = RigVMDispatchNode(Array=RigVMVariableNode() {variable=Physics Owners}.Value, Index=RigVMVariableNode() {variable=Index A}.Value) {template="DISPATCH_RigVMDispatch_ArrayGetAtIndex(in Array,in Index,out Element)"}
   Get_Metadata_3 = RigVMDispatchNode(Default=PhysicsBody, Name=Has Physics Body, NameSpace=None, Item=@For_Each_3.Element) {template="DISPATCH_RigDispatch_GetMetadata(in Item,in Name,in NameSpace,in Default,out Value,out Found)"}
   Get_Metadata_4 = RigVMDispatchNode(Default=PhysicsBody, Name=Has Physics Body, NameSpace=None, Item=@n17.Element) {template="DISPATCH_RigDispatch_GetMetadata(in Item,in Name,in NameSpace,in Default,out Value,out Found)"}
   > &RigUnit_HierarchyDisableCollisionBetween(
    PhysicsBodyComponentKey1=(ElementKey=(Type=Bone,Name="None"),Name="PhysicsBody")
    PhysicsBodyComponentKey2=(ElementKey=(Type=Bone,Name="None"),Name="PhysicsBody")
    PhysicsBodyComponentKey1.ElementKey=@n17.Element
    PhysicsBodyComponentKey2.Name=@Get_Metadata_3.Value
    PhysicsBodyComponentKey1.Name=@Get_Metadata_4.Value
    PhysicsBodyComponentKey2.ElementKey=@For_Each_3.Element
   )
   RigVMVariableNode(
    Value=/Script/ControlRig.RigUnit_FindItemsWithMetadata(Name=Has Physics Body, NameSpace=None, Type=Name) {is=struct RigVMUnitNode}.Items
    ExecuteContext=@RigVMFunction_Sequence.A
   ) {variable=Physics Owners}
   For_Each_7 = RigVMDispatchNode(Array=(), ExecuteContext=()) {template="DISPATCH_RigVMDispatch_ArrayIterator(in Array,out Element,out Index,out Count,out Ratio)"}
   RigVMFunction_MathBoolToInteger = &RigVMFunction_MathBoolToInteger(Value=@n1.Value)
   RigVMDispatchNode(ExecuteContext=(), Index=@RigVMFunction_MathBoolToInteger.Result) {template="DISPATCH_RigVMDispatch_SwitchInt32(in Index)"}
  ]}
]
hierarchy [
 Root {type=Connector, loc=<0 0 0>, rot=<0 0 0 1>}
 Root {type=Socket, loc=<0 0 0>, rot=<0 0 0 1>}
 Disable Bones B {type=Connector, loc=<0 0 0>, rot=<0 0 0 1>}
 Disable Bones A {type=Connector, loc=<0 0 0>, rot=<0 0 0 1>}
]
variables [
 Excluded Physics Owners {type="TArray<FRigElementKey>"}
 Index A {type=int32, default=0}
 Index B {type=int32, default=0}
 Physics Owners {type="TArray<FRigElementKey>"}
 Remove All Collisions {type=bool, default=False}
]
