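# Control Rig module that provides a kinematic ground collider following the slope beneath a character for runtime physics.
# Construction copies the parent module's solver metadata, spawns or reuses a bone at the Root connector, and optionally adds a box body offset down by half its height.
# Each update places the bone from world-space slope translation and rotation inputs, converted to rig space, dropping it 200 units while Is In Air is true.
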ControlRigBlueprint /EpicControlRigRuntime/CRM_Runtime_Physics_Ground_v01
graphs [
 ConstructionGraph {nodes [
   RigUnit_PrepareForExecution = &RigUnit_PrepareForExecution()
   RigVMFunction_Sequence = RigVMAggregateNode(ExecuteContext=@RigUnit_PrepareForExecution.ExecutePin)
   RigVMFunction_ControlFlowBranch = &RigVMFunction_ControlFlowBranch(ExecuteContext=@RigVMFunction_Sequence.B, Condition=RigVMVariableNode() {variable=Spawn Bone}.Value)
   RigVMVariableNode(
    Value=/Script/ControlRig.RigUnit_ResolveConnector(Connector=(Type=Connector,Name="Root"), SkipSocket=true) {is=struct RigVMUnitNode}.Result
    ExecuteContext=@RigVMFunction_ControlFlowBranch.False
   ) {variable=Bone}
   n1 = /Script/ControlRig.RigUnit_HierarchyAddBone(
    Name=ground
    Space=LocalSpace
    Transform=(Rotation=(X=0,Y=0,Z=0,W=1),Translation=(X=0,Y=0,Z=0),Scale3D=(X=1,Y=1,Z=1))
    Parent=&RigUnit_ResolveConnector(Connector=(Type=Connector,Name="Root"), SkipSocket=true).Result
    ExecutePin=@RigVMFunction_ControlFlowBranch.True
   ) {is=struct RigVMUnitNode}
   > RigVMVariableNode(Value=@n1.Item) {variable=Bone}
   n2 = &RigVMFunction_ControlFlowBranch(ExecuteContext=@RigVMFunction_Sequence.C, Condition=RigVMVariableNode() {variable=Spawn Collider}.Value)
   n3 = RigVMVariableNode() {variable=Extends}
   Negate = /Script/RigVM.RigVMFunction_MathDoubleNegate(Value=&RigVMFunction_MathDoubleMul(B=0.5, A=@n3.Value.Z).Result) {is=struct RigVMUnitNode, template="Negate::Execute(in Value,out Result)"}
   n4 = /Script/ControlRigPhysics.RigUnit_AddPhysicsBody(
    BodyData=(MovementType=Kinematic,CollisionType=QueryAndPhysics,KinematicTargetSpace=OffsetInBoneSpace,GravityMultiplier=1,PhysicsBlendWeight=1,bUpdateKinematicFromSimulation=True,bEnableCCD=False)
    Collision="(Boxes=((RestOffset=0,Name=\"Ground\",bContributeToMass=false,TM=(Rotation=(X=0,Y=0,Z=0,W=1),Translation=(X=0,Y=0,Z=0),Scale3D=(X=1,Y=1,Z=1)),Extents=(X=10,Y=10,Z=10))),Material=(Friction=0.5,Restitution=0,FrictionCombineMode=Multiply,RestitutionCombineMode= ...(truncated)"
    Dynamics="(Density=1,MassOverride=1,bOverrideCentreOfMass=False,CentreOfMassOverride=(X=0,Y=0,Z=0),CentreOfMassNudge=(X=0,Y=0,Z=0),bOverrideMomentsOfInertia=False,MomentsOfInertiaOverride=(X=1,Y=1,Z=1),LinearDamping=0,AngularDamping=0,bScaleDampingByInverseMass=Fals ...(truncated)"
    Solver=(PhysicsSolverComponentKey=(ElementKey=(Type=Bone,Name="None"),Name="PhysicsBody"),bUseAutomaticSolver=True,SourceBone=(Type=None,Name="None"),TargetBone=(Type=None,Name="None"),bIncludeInChecksForReset=True)
    Collision.Boxes.0.TM.Translation.Z=@Negate.Result
    ExecutePin=@n2.True
    Collision.Boxes.0.Extents=@n3.Value
    Collision.Material=RigVMVariableNode() {variable=Material}.Value
    Collision.Boxes.0.RestOffset=RigVMVariableNode() {variable=Rest Offset}.Value
    Solver.PhysicsSolverComponentKey=RigVMVariableNode() {variable=Solver}.Value
    Owner=RigVMVariableNode() {variable=Bone}.Value
   ) {is=struct RigVMUnitNode}
   > Set_Metadata = RigVMDispatchNode(Name=Has Physics Body, NameSpace=None, Item=@n4.PhysicsBodyComponentKey.ElementKey, Value=@n4.PhysicsBodyComponentKey.Name) {template="DISPATCH_RigDispatch_SetMetadata(in Item,in Name,in NameSpace,in Value,out Success)"}
   Get_Module_Metadata = RigVMDispatchNode(Default=(Type=Bone,Name="None"), Name=Physics Solver Owner, NameSpace=Parent) {template="DISPATCH_RigDispatch_GetModuleMetadata(in Name,in NameSpace,in Default,out Value,out Found)"}
   Get_Module_Metadata_1 = RigVMDispatchNode(Default=PhysicsSolver, Name=Physics Solver Name, NameSpace=Parent) {template="DISPATCH_RigDispatch_GetModuleMetadata(in Name,in NameSpace,in Default,out Value,out Found)"}
   RigVMVariableNode(
    Value=(ElementKey=(Type=None,Name="None"),Name="None")
    Value.ElementKey=@Get_Module_Metadata.Value
    Value.Name=@Get_Module_Metadata_1.Value
    ExecuteContext=@RigVMFunction_Sequence.A
   ) {variable=Solver}
   RigVMFunction_MathBoolAnd = &RigVMFunction_MathBoolAnd(A=@Get_Module_Metadata.Found, B=@Get_Module_Metadata_1.Found)
   > RigVMVariableNode(Value=@RigVMFunction_MathBoolAnd.Result) {variable=Solver Exists}
   > Set_Module_Metadata = RigVMDispatchNode(Name=Physics Solver Owner, NameSpace=Self, Value=@Get_Module_Metadata.Value) {template="DISPATCH_RigDispatch_SetModuleMetadata(in Name,in NameSpace,in Value,out Success)"}
   > Set_Module_Metadata_1 = RigVMDispatchNode(Name=Physics Solver Name, NameSpace=Self, Value=@Get_Module_Metadata_1.Value) {template="DISPATCH_RigDispatch_SetModuleMetadata(in Name,in NameSpace,in Value,out Success)"}
  ]}
 RigVMFunctionLibrary {links [], nodes []}
 RigVMFunction_Sequence_1_ContainedGraph {nodes [
   n1 = RigVMFunctionEntryNode()
   RigVMFunction_Sequence = &RigVMFunction_Sequence(ExecuteContext=@n1.ExecuteContext)
   n2 = &RigVMFunction_Sequence(ExecuteContext=@RigVMFunction_Sequence.B)
   RigVMFunctionReturnNode(A=@RigVMFunction_Sequence.A, B=@n2.A, C=@n2.B)
  ]}
 RigVMModel {nodes [
   RigUnit_BeginExecution = &RigUnit_BeginExecution()
   n1 = RigVMVariableNode() {variable=Root Slope Translation}
   Add = /Script/RigVM.RigVMFunction_MathVectorAdd(B=(X=0,Y=0,Z=-200), A=@n1.Value) {is=struct RigVMUnitNode, template="Add::Execute(in A,in B,out Result)"}
   RigVMFunction_MathBoolToFloat = &RigVMFunction_MathBoolToFloat(Value=RigVMVariableNode() {variable=Is In Air}.Value)
   Interpolate = /Script/RigVM.RigVMFunction_MathVectorLerp(B=@Add.Result, T=@RigVMFunction_MathBoolToFloat.Result, A=@n1.Value) {is=struct RigVMUnitNode, template="Interpolate::Execute(in A,in B,in T,out Result)"}
   From World = /Script/ControlRig.RigUnit_ToRigSpace_Location(Value=@Interpolate.Result) {is=struct RigVMUnitNode, template="From World::Execute(in Value,out Global)"}
   RigVMFunction_MathQuaternionFromRotatorV2 = /Script/RigVM.RigVMFunction_MathQuaternionFromRotatorV2(Value=RigVMVariableNode() {variable=Slope Rotation}.Value) {is=struct RigVMUnitNode, template="Cast::Execute(in Value,out Result)"}
   From World_1 = /Script/ControlRig.RigUnit_ToRigSpace_Rotation(Value=@RigVMFunction_MathQuaternionFromRotatorV2.Result) {is=struct RigVMUnitNode, template="From World::Execute(in Value,out Global)"}
   &RigUnit_SetTransform(
    Value=(Rotation=(X=0,Y=0,Z=0,W=1),Translation=(X=0,Y=0,Z=0),Scale3D=(X=1,Y=1,Z=1))
    Weight=1
    bInitial=False
    Value.Translation=@"From World".Global
    Value.Rotation=@"From World_1".Global
    ExecutePin=@RigUnit_BeginExecution.ExecutePin
    Item=RigVMVariableNode() {variable=Bone}.Value
   )
  ]}
]
hierarchy [
 Root {type=Connector, loc=<0 0 0>, rot=<0 0 0 1>}
 Root {type=Socket, loc=<0 0 0>, rot=<0 0 0 1>}
]
variables [
 Bone {type=FRigElementKey, default=(Type=Bone,Name="")}
 Extends {type=FVector, default=(X=150,Y=150,Z=20)}
 Is In Air {type=bool, default=False}
 Material {type=FRigPhysicsMaterial, default=(Friction=0.5,Restitution=0,FrictionCombineMode=Multiply,RestitutionCombineMode=Multiply)}
 Rest Offset {type=float, default=0}
 Root Slope Translation {type=FVector, default=(X=0,Y=0,Z=0)}
 Slope Rotation {type=FRotator, default=(Pitch=0,Yaw=0,Roll=0)}
 Solver {type=FRigComponentKey, default=(ElementKey=(Type=Bone,Name=""),Name="PhysicsSolver")}
 Solver Exists {type=bool, default=False}
 Spawn Bone {type=bool, default=True}
 Spawn Collider {type=bool, default=True}
]
