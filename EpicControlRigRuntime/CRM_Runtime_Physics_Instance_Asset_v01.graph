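# Control Rig module that builds physics bodies and joints from an array of physics assets at construction.
# If the Owner connector is connected, it reads the parent module's physics solver metadata into Solver and republishes it on itself.
# It then instantiates each asset against that solver with helper bodies, using bone-list, joint and drive settings.
# Resulting component names are written as "Has Physics Body" and "Has Physics Joint" metadata on their elements.
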ControlRigBlueprint /EpicControlRigRuntime/CRM_Runtime_Physics_Instance_Asset_v01
graphs [
 RigVMFunctionLibrary {links [], nodes []}
 RigVMModel {nodes [
   n1 = &RigUnit_PrepareForExecution()
   RigVMFunction_Sequence = &RigVMFunction_Sequence(ExecuteContext=@n1.ExecutePin)
   RigVMFunction_ControlFlowBranch = &RigVMFunction_ControlFlowBranch(
    Condition=/Script/ControlRig.RigUnit_ResolveConnector(Connector=(Type=Connector,Name="Owner"), SkipSocket=true) {is=struct RigVMUnitNode}.bIsConnected
    ExecuteContext=@RigVMFunction_Sequence.A
   )
   Get_Module_Metadata = RigVMDispatchNode(Default=(Type=Bone,Name="None"), Name=Physics Solver Owner, NameSpace=Parent) {template="DISPATCH_RigDispatch_GetModuleMetadata(in Name,in NameSpace,in Default,out Value,out Found)"}
   Get_Module_Metadata_1 = RigVMDispatchNode(Default=PhysicsSolver, Name=Physics Solver Name, NameSpace=Parent) {template="DISPATCH_RigDispatch_GetModuleMetadata(in Name,in NameSpace,in Default,out Value,out Found)"}
   RigVMVariableNode(
    Value=(ElementKey=(Type=None,Name="None"),Name="None")
    Value.ElementKey=@Get_Module_Metadata.Value
    Value.Name=@Get_Module_Metadata_1.Value
    ExecuteContext=@RigVMFunction_ControlFlowBranch.True
   ) {variable=Solver}
   > RigVMVariableNode(Value=&RigVMFunction_MathBoolAnd(A=@Get_Module_Metadata.Found, B=@Get_Module_Metadata_1.Found).Result) {variable=Solver Exists}
   > Set_Module_Metadata = RigVMDispatchNode(Name=Physics Solver Owner, NameSpace=Self, Value=@Get_Module_Metadata.Value) {template="DISPATCH_RigDispatch_SetModuleMetadata(in Name,in NameSpace,in Value,out Success)"}
   > Set_Module_Metadata_1 = RigVMDispatchNode(Name=Physics Solver Name, NameSpace=Self, Value=@Get_Module_Metadata_1.Value) {template="DISPATCH_RigDispatch_SetModuleMetadata(in Name,in NameSpace,in Value,out Success)"}
   For_Each = RigVMDispatchNode(ExecuteContext=@RigVMFunction_Sequence.B, Array=RigVMVariableNode() {variable=Physics Asset}.Value) {template="DISPATCH_RigVMDispatch_ArrayIterator(in Array,out Element,out Index,out Count,out Ratio)"}
   > n2 = /Script/ControlRigPhysics.RigUnit_HierarchyInstantiateFromPhysicsAsset(
    ConstraintProfileName=None
    ParentSpaceControlData="(bEnabled=True,LinearStrength=0,LinearDampingRatio=1,LinearExtraDamping=0,MaxForce=0,AngularStrength=0,AngularDampingRatio=1,AngularExtraDamping=0,MaxTorque=0,LinearTargetVelocityMultiplier=1,AngularTargetVelocityMultiplier=1,CustomControlPoint=(X=0,Y=0,Z= ...(truncated)"
    SimSpaceControlData="(bEnabled=True,LinearStrength=0,LinearDampingRatio=1,LinearExtraDamping=0,MaxForce=0,AngularStrength=0,AngularDampingRatio=1,AngularExtraDamping=0,MaxTorque=0,LinearTargetVelocityMultiplier=1,AngularTargetVelocityMultiplier=1,CustomControlPoint=(X=0,Y=0,Z= ...(truncated)"
    bAddParentSpaceControl=false
    bAddSimSpaceControl=false
    bMakeHelperBodies=True
    bUseAutomaticSolver=false
    BonesToUse=RigVMVariableNode() {variable=Bones To Use}.Value
    bEnableJoints=RigVMVariableNode() {variable=Enable Joints}.Value
    PhysicsSolverComponentKey=RigVMVariableNode() {variable=Solver}.Value
    bEnableDrives=RigVMVariableNode() {variable=Enable Drives}.Value
    PhysicsAsset=@For_Each.Element
   ) {is=struct RigVMUnitNode}
   > For_Each_1 = RigVMDispatchNode(Array=@n2.PhysicsBodyComponentKeys) {template="DISPATCH_RigVMDispatch_ArrayIterator(in Array,out Element,out Index,out Count,out Ratio)"}
   > Set_Metadata = RigVMDispatchNode(Name=Has Physics Body, NameSpace=None, Item=@For_Each_1.Element.ElementKey, Value=@For_Each_1.Element.Name) {template="DISPATCH_RigDispatch_SetMetadata(in Item,in Name,in NameSpace,in Value,out Success)"}
   For_Each_2 = RigVMDispatchNode(ExecuteContext=@For_Each_1.Completed, Array=@n2.PhysicsJointComponentKeys) {template="DISPATCH_RigVMDispatch_ArrayIterator(in Array,out Element,out Index,out Count,out Ratio)"}
   > Set_Metadata_1 = RigVMDispatchNode(Name=Has Physics Joint, NameSpace=None, Item=@For_Each_2.Element.ElementKey, Value=@For_Each_2.Element.Name) {template="DISPATCH_RigDispatch_SetMetadata(in Item,in Name,in NameSpace,in Value,out Success)"}
  ]}
]
hierarchy [
 Owner {type=Connector, loc=<0 0 0>, rot=<0 0 0 1>}
 Root {type=Socket, loc=<0 0 0>, rot=<0 0 0 1>}
]
variables [
 Bones To Use {type="TArray<FRigElementKey>"}
 Enable Drives {type=bool, default=True}
 Enable Joints {type=bool, default=True}
 Owner {type=FRigElementKey, default=(Type=None,Name="")}
 Physics Asset {type="TArray<TObjectPtr<UPhysicsAsset>>"}
 Solver {type=FRigComponentKey, default=(ElementKey=(Type=None,Name=""),Name="")}
 Solver Exists {type=bool, default=False}
]
