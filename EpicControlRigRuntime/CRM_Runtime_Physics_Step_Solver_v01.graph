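# Control Rig module that steps a physics solver at runtime, with adjustable alpha, delta time overrides and visualization settings.
# On construction, it reads the solver's owner and name from parent-namespace module metadata and stores the component key and an existence flag.
# Each update, if the solver exists, it optionally applies simulation-space overrides (clamps, drag, external velocities, space movement), then steps the solver.
# A separate Set External Velocity node is left unconnected.
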ControlRigBlueprint /EpicControlRigRuntime/CRM_Runtime_Physics_Step_Solver_v01
graphs [
 RigVMFunctionLibrary {links [], nodes []}
 RigVMModel {nodes [
   n1 = &RigUnit_PrepareForExecution()
   Get_Module_Metadata_2 = RigVMDispatchNode(Default=(Type=Bone,Name="root"), Name=Physics Solver Owner, NameSpace=Parent) {template="DISPATCH_RigDispatch_GetModuleMetadata(in Name,in NameSpace,in Default,out Value,out Found)"}
   Get_Module_Metadata_3 = RigVMDispatchNode(Default=PhysicsSolver, Name=Physics Solver Name, NameSpace=Parent) {template="DISPATCH_RigDispatch_GetModuleMetadata(in Name,in NameSpace,in Default,out Value,out Found)"}
   RigVMVariableNode(
    ExecuteContext=@n1.ExecutePin
    Value=&RigVMFunction_MathBoolAnd(A=@Get_Module_Metadata_2.Found, B=@Get_Module_Metadata_3.Found).Result
   ) {variable=Solver Exists}
   > RigVMVariableNode(
    Value=(ElementKey=(Type=None,Name="None"),Name="None")
    Value.ElementKey=@Get_Module_Metadata_2.Value
    Value.Name=@Get_Module_Metadata_3.Value
   ) {variable=Solver}
   RigUnit_BeginExecution = &RigUnit_BeginExecution()
   RigVMFunction_ControlFlowBranch = &RigVMFunction_ControlFlowBranch(ExecuteContext=@RigUnit_BeginExecution.ExecutePin, Condition=RigVMVariableNode() {variable=Solver Exists}.Value)
   n2 = &RigVMFunction_ControlFlowBranch(ExecuteContext=@RigVMFunction_ControlFlowBranch.True, Condition=RigVMVariableNode() {variable=Enable Overrides}.Value)
   /Script/ControlRigPhysics.RigUnit_StepPhysicsSolver(
    ExecutePin=@n2.Completed
    DeltaTimeOverride=RigVMVariableNode() {variable=Delta Time Override}.Value
    SimulationSpaceDeltaTimeOverride=RigVMVariableNode() {variable=Simulation Space Delta Time Override}.Value
    Alpha=RigVMVariableNode() {variable=Alpha}.Value
    VisualizationSettings=RigVMVariableNode() {variable=Visualization Settings}.Value
    PhysicsSolverComponentKey=RigVMVariableNode() {variable=Solver}.Value
   ) {is=struct RigVMUnitNode}
   /Script/ControlRigPhysics.RigUnit_SetPhysicsSolverSimulationSpaceSettings(
    SimulationSpaceSettings="(SpaceMovementAmount=1,VelocityScaleZ=1,bClampLinearVelocity=False,MaxLinearVelocity=10000,bClampAngularVelocity=False,MaxAngularVelocity=10000,bClampLinearAcceleration=False,MaxLinearAcceleration=10000,bClampAngularAcceleration=False,MaxAngularAcceleratio ...(truncated)"
    ExecutePin=@n2.True
    SimulationSpaceSettings.bClampLinearVelocity=RigVMVariableNode() {variable=Clamp Linear Velocity}.Value
    SimulationSpaceSettings.MaxLinearVelocity=RigVMVariableNode() {variable=Max Linear Velocity}.Value
    SimulationSpaceSettings.bClampAngularVelocity=RigVMVariableNode() {variable=Clamp Angular Velocity}.Value
    SimulationSpaceSettings.MaxAngularVelocity=RigVMVariableNode() {variable=Max Angular Velocity}.Value
    SimulationSpaceSettings.bClampLinearAcceleration=RigVMVariableNode() {variable=Clamp Linear Acceleration}.Value
    SimulationSpaceSettings.MaxLinearAcceleration=RigVMVariableNode() {variable=Max Linear Acceleration}.Value
    SimulationSpaceSettings.bClampAngularAcceleration=RigVMVariableNode() {variable=Clamp Angular Acceleration}.Value
    SimulationSpaceSettings.MaxAngularAcceleration=RigVMVariableNode() {variable=Max Angular Acceleration}.Value
    SimulationSpaceSettings.LinearDragMultiplier=RigVMVariableNode() {variable=Linear Drag Multiplier}.Value
    SimulationSpaceSettings.AngularDragMultiplier=RigVMVariableNode() {variable=Angular Drag Multiplier}.Value
    SimulationSpaceSettings.ExternalLinearDrag=RigVMVariableNode() {variable=External Linear Drag}.Value
    SimulationSpaceSettings.ExternalLinearVelocity=RigVMVariableNode() {variable=External Linear Velocity}.Value
    SimulationSpaceSettings.ExternalAngularVelocity=RigVMVariableNode() {variable=External Angular Velocity}.Value
    PhysicsSolverComponentKey=RigVMVariableNode() {variable=Solver}.Value
    SimulationSpaceSettings.SpaceMovementAmount=RigVMVariableNode() {variable=Space Movement Amount}.Value
    SimulationSpaceSettings.VelocityScaleZ=RigVMVariableNode() {variable=Velocity Scale Z}.Value
   ) {is=struct RigVMUnitNode}
   /Script/ControlRigPhysics.RigUnit_SetPhysicsSolverExternalVelocity(
    ExecutePin=()
    ExternalAngularVelocity=(X=0,Y=0,Z=0)
    ExternalLinearVelocity=(X=0,Y=0,Z=0)
    ExternalTurbulenceVelocity=(X=0,Y=0,Z=0)
    PhysicsSolverComponentKey=(ElementKey=(Type=Bone,Name="None"),Name="PhysicsSolver")
   ) {is=struct RigVMUnitNode}
  ]}
]
hierarchy [
 Root {type=Connector, loc=<0 0 0>, rot=<0 0 0 1>}
 Root {type=Socket, loc=<0 0 0>, rot=<0 0 0 1>}
]
variables [
 Alpha {type=float, default=1}
 Angular Drag Multiplier {type=float, default=1}
 Clamp Angular Acceleration {type=bool, default=True}
 Clamp Angular Velocity {type=bool, default=True}
 Clamp Linear Acceleration {type=bool, default=True}
 Clamp Linear Velocity {type=bool, default=True}
 Delta Time Override {type=float, default=0}
 Enable Overrides {type=bool, default=False}
 External Angular Velocity {type=FVector, default=(X=0,Y=0,Z=0)}
 External Linear Drag {type=FVector, default=(X=0,Y=0,Z=0)}
 External Linear Velocity {type=FVector, default=(X=0,Y=0,Z=0)}
 Linear Drag Multiplier {type=float, default=1}
 Max Angular Acceleration {type=float, default=48}
 Max Angular Velocity {type=float, default=12}
 Max Linear Acceleration {type=float, default=2800}
 Max Linear Velocity {type=float, default=700}
 Simulation Space Delta Time Override {type=float, default=0}
 Solver {type=FRigComponentKey, default=(ElementKey=(Type=Bone,Name=""),Name="PhysicsSolver")}
 Solver Exists {type=bool, default=False}
 Space Movement Amount {type=float, default=1}
 Velocity Scale Z {type=float, default=1}
 Visualization Settings {
  type=FRigPhysicsVisualizationSettings
  default=(bEnableVisualization=True,LineThickness=0.1,ShapeSize=1,ShapeDetail=16,bShowBodies=True,bShowCentreOfMass=False,bShowJoints=True,bShowControls=False,bShowWorldObjects=True,bShowWorldOverlapBox=False,bShowActiveContacts=True,bShowInactiveContacts=True)
 }
]
