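# Modular Control Rig slope-warping module that fits a biped's feet and pelvis to sloped ground.
# Construction resolves connectors and measures leg lengths.
# While grounded, moving below a speed threshold, it samples ground from supplied slope data, one pelvis sphere trace projected per foot, or per-foot traces.
# Smoothed offsets tilt feet, shift the pelvis by the lower offset and drive two-bone IK, weighted by Alpha and a DisableSlopeWarping curve.
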ControlRigBlueprint /EpicControlRigRuntime/CRM_Runtime_SlopeWarping_v01
graphs [
 Add_2_ContainedGraph {nodes [
   n1 = RigVMFunctionEntryNode()
   Add = &RigVMFunction_MathDoubleAdd(A=@n1.A, B=@n1.B)
   RigVMFunctionReturnNode(Result=&RigVMFunction_MathDoubleAdd(A=@Add.Result, B=@n1.C).Result)
  ]}
 Add_2_ContainedGraph {nodes [
   n1 = RigVMFunctionEntryNode()
   RigVMFunctionReturnNode(Result=&RigVMFunction_MathDoubleAdd(A=&RigVMFunction_MathDoubleAdd(A=@n1.A, B=@n1.B).Result, B=@n1.C).Result)
  ]}
 ConstructionGraph {nodes [
   RigUnit_PrepareForExecution = &RigUnit_PrepareForExecution()
   n1 = &RigVMFunction_Sequence(ExecuteContext=@RigUnit_PrepareForExecution.ExecutePin)
   RigVMFunction_Sequence_2 = RigVMAggregateNode(ExecuteContext=@n1.A)
   n2 = /Script/ControlRig.RigUnit_ResolveConnector(Connector=(Type=Connector,Name="Foot R"), SkipSocket=False) {is=struct RigVMUnitNode}
   RigVMVariableNode(Value=@n2.Result, ExecuteContext=@RigVMFunction_Sequence_2.B) {variable=Foot R}
   n3 = /Script/ControlRig.RigUnit_HierarchyGetParent(bDefaultParent=True, Child=@n2.Result) {is=struct RigVMUnitNode}
   > RigVMVariableNode(Value=@n3.Parent) {variable=Calf R}
   > RigVMVariableNode(Value=&RigUnit_HierarchyGetParent(bDefaultParent=True, Child=@n3.Parent).Parent) {variable=Thigh R}
   n4 = &RigUnit_ResolveConnector(Connector=(Type=Connector,Name="Foot L"), SkipSocket=False)
   RigVMVariableNode(Value=@n4.Result, ExecuteContext=@RigVMFunction_Sequence_2.A) {variable=Foot L}
   n5 = &RigUnit_HierarchyGetParent(bDefaultParent=True, Child=@n4.Result)
   > RigVMVariableNode(Value=@n5.Parent) {variable=Calf L}
   > RigVMVariableNode(Value=&RigUnit_HierarchyGetParent(bDefaultParent=True, Child=@n5.Parent).Parent) {variable=Thigh L}
   RigVMVariableNode(
    Value=&RigUnit_ResolveConnector(Connector=(Type=Connector,Name="Pelvis"), SkipSocket=False).Result
    ExecuteContext=@RigVMFunction_Sequence_2.C
   ) {variable=Pelvis}
   > RigVMVariableNode(Value=&RigUnit_ResolveConnector(Connector=(Type=Connector,Name="Root"), SkipSocket=False).Result) {variable=Root}
   RigVMFunction_Sequence = &RigVMFunction_Sequence(ExecuteContext=@n1.B)
   Leg Segment Lengths = RigVMFunctionReferenceNode(
    Thigh=RigVMVariableNode() {variable=Thigh L}.Value
    Calf=RigVMVariableNode() {variable=Calf L}.Value
    Foot=RigVMVariableNode() {variable=Foot L}.Value
   )
   RigVMVariableNode(Value=@"Leg Segment Lengths".Bone A Length, ExecuteContext=@RigVMFunction_Sequence.A) {variable=Bone A Length L}
   > RigVMVariableNode(Value=@"Leg Segment Lengths".Bone B Length) {variable=Bone B Length L}
   > RigVMVariableNode(Value=@"Leg Segment Lengths".Foot Height) {variable=Height Foot L}
   Leg Segment Lengths_1 = RigVMFunctionReferenceNode(
    Foot=RigVMVariableNode() {variable=Foot R}.Value
    Thigh=RigVMVariableNode() {variable=Thigh R}.Value
    Calf=RigVMVariableNode() {variable=Calf R}.Value
   )
   RigVMVariableNode(Value=@"Leg Segment Lengths_1".Bone A Length, ExecuteContext=@RigVMFunction_Sequence.B) {variable=Bone A Length R}
   > RigVMVariableNode(Value=@"Leg Segment Lengths_1".Bone B Length) {variable=Bone B Length R}
   > RigVMVariableNode(Value=@"Leg Segment Lengths_1".Foot Height) {variable=Height Foot R}
  ]}
 ContainedGraph {nodes [
   RigVMFunctionEntryNode()
   > RigVMFunction_Sequence = RigVMAggregateNode()
   RigVMFunction_ControlFlowBranch = &RigVMFunction_ControlFlowBranch(ExecuteContext=@RigVMFunction_Sequence.C, Condition=RigVMVariableNode() {variable=Debug Draw}.Value)
   RigVMFunction_DebugTransformArrayMutableNoSpace = /Script/RigVM.RigVMFunction_DebugTransformArrayMutableNoSpace(
    Color=(R=1,G=1,B=0,A=1)
    DebugDrawSettings=(DepthPriority=SDPG_Foreground,Lifetime=-1)
    Mode=Point
    ParentIndices=()
    Scale=3
    Thickness=1
    Transforms=((Rotation=(X=0,Y=0,Z=0,W=1),Translation=(X=0,Y=0,Z=0),Scale3D=(X=1,Y=1,Z=1)),(Rotation=(X=0,Y=0,Z=0,W=1),Translation=(X=0,Y=0,Z=0),Scale3D=(X=1,Y=1,Z=1)))
    WorldOffset=(Rotation=(X=0,Y=0,Z=0,W=1),Translation=(X=0,Y=0,Z=0),Scale3D=(X=1,Y=1,Z=1))
    ExecutePin=@RigVMFunction_ControlFlowBranch.True
    Transforms.0.Translation=RigVMVariableNode() {variable=Start Trace}.Value
    Transforms.1.Translation=RigVMVariableNode() {variable=End Trace}.Value
    bEnabled=RigVMVariableNode() {variable=Debug Draw}.Value
   ) {is=struct RigVMUnitNode}
   > RigVMFunction_DebugLineNoSpace = /Script/RigVM.RigVMFunction_DebugLineNoSpace(
    Color=(R=1,G=1,B=0,A=1)
    DebugDrawSettings=(DepthPriority=SDPG_Foreground,Lifetime=-1)
    Thickness=0.75
    WorldOffset=(Rotation=(X=0,Y=0,Z=0,W=1),Translation=(X=0,Y=0,Z=0),Scale3D=(X=1,Y=1,Z=1))
    A=RigVMVariableNode() {variable=Start Trace}.Value
    B=RigVMVariableNode() {variable=End Trace}.Value
    bEnabled=RigVMVariableNode() {variable=Debug Draw}.Value
   ) {is=struct RigVMUnitNode}
   > RigVMFunction_DebugTransformMutableNoSpace = /Script/RigVM.RigVMFunction_DebugTransformMutableNoSpace(
    Color=(R=1,G=0,B=0,A=1)
    DebugDrawSettings=(DepthPriority=SDPG_Foreground,Lifetime=-1)
    Mode=Point
    Scale=4
    Thickness=1
    Transform=(Rotation=(X=0,Y=0,Z=0,W=1),Translation=(X=0,Y=0,Z=0),Scale3D=(X=1,Y=1,Z=1))
    WorldOffset=(Rotation=(X=0,Y=0,Z=0,W=1),Translation=(X=0,Y=0,Z=0),Scale3D=(X=1,Y=1,Z=1))
    Transform.Translation=RigVMVariableNode() {variable=Position Offset}.Value
    bEnabled=RigVMVariableNode() {variable=Debug Draw}.Value
   ) {is=struct RigVMUnitNode}
   n1 = &RigUnit_GetTransform(Item=RigVMVariableNode() {variable=Foot Bone}.Value)
   n2 = &RigUnit_GetTransform(Item=RigVMVariableNode() {variable=Root Bone}.Value)
   n3 = &RigVMFunction_MathFloatAdd(A=@n1.Transform.Translation.Y, B=RigVMVariableNode() {variable=Trace Offset}.Value) {template="Add::Execute(in A,in B,out Result)"}
   n4 = RigVMVariableNode() {variable=Trace Distance}
   RigVMVariableNode(
    Value=/Script/RigVM.RigVMFunction_MathVectorAdd(A=(X=0,Y=0,Z=0), B=(X=0,Y=0,Z=20), A.X=@n1.Transform.Translation.X, A.Z=@n2.Transform.Translation.Z, A.Y=@n3.Result, B.Z=@n4.Value) {is=struct RigVMUnitNode, template="Add::Execute(in A,in B,out Result)"}.Result
    ExecuteContext=@RigVMFunction_Sequence.A
   ) {variable=Start Trace}
   > RigVMVariableNode(Value=/Script/RigVM.RigVMFunction_MathVectorSub(A=(X=0,Y=0,Z=0), B=(X=0,Y=0,Z=30), A.X=@n1.Transform.Translation.X, A.Z=@n2.Transform.Translation.Z, A.Y=@n3.Result, B.Z=@n4.Value) {is=struct RigVMUnitNode, template="Subtract::Execute(in A,in B,out Result)"}.Result) {variable=End Trace}
   n5 = /Script/ControlRig.RigUnit_SphereTraceByTraceChannel(
    Radius=RigVMVariableNode() {variable=Trace Radius}.Value
    Start=RigVMVariableNode() {variable=Start Trace}.Value
    End=RigVMVariableNode() {variable=End Trace}.Value
    TraceChannel=RigVMVariableNode() {variable=Trace Channel}.Value
   ) {is=struct RigVMUnitNode}
   RigVMVariableNode(Value=@n5.bHit, ExecuteContext=@RigVMFunction_Sequence.B) {variable=Is Hit}
   n6 = /Script/ControlRig.RigUnit_VisualDebugVectorItemSpace(
    Color=(R=1,G=0,B=0,A=1)
    DebugDrawSettings=(DepthPriority=SDPG_Foreground,Lifetime=-1)
    Scale=1
    Space=(Type=Bone)
    Thickness=10
    bEnabled=True
    Value=@n5.HitLocation
   ) {is=struct RigVMUnitNode}
   > RigVMVariableNode(Value=@n6.Value) {variable=Position Offset}
   > RigVMVariableNode(Value=@n5.HitNormal) {variable=Normal}
   RigVMFunction_MathQuaternionFromTwoVectors = /Script/RigVM.RigVMFunction_MathQuaternionFromTwoVectors(A=(X=0,Y=0,Z=1), B=@n5.HitNormal) {is=struct RigVMUnitNode, template="FromTwoVectors::Execute(in A,in B,out Result)"}
   > RigVMVariableNode(Value=@RigVMFunction_MathQuaternionFromTwoVectors.Result) {variable=Rotation Offset}
   RigVMFunctionReturnNode(
    Hit Normal=RigVMVariableNode() {variable=Normal}.Value
    Hit Position=RigVMVariableNode() {variable=Position Offset}.Value
    Hit Rotation=RigVMVariableNode() {variable=Rotation Offset}.Value
    Hit=RigVMVariableNode() {variable=Is Hit}.Value
    ExecuteContext=@RigVMFunction_Sequence.D
   )
  ]}
 New Function_ContainedGraph {nodes [
   n1 = RigVMFunctionEntryNode()
   n2 = &RigUnit_GetTransform(bInitial=true, Item=@n1.Foot)
   n3 = &RigUnit_GetTransform(bInitial=true, Item=@n1.Calf)
   RigVMFunction_MathVectorDistance = &RigVMFunction_MathVectorDistance(B=&RigUnit_GetTransform(bInitial=true, Item=@n1.Thigh).Transform.Translation, A=@n3.Transform.Translation)
   > RigVMFunctionReturnNode(
    Foot Height=@n2.Transform.Translation.Z
    Bone A Length=@RigVMFunction_MathVectorDistance.Result
    Bone B Length=&RigVMFunction_MathVectorDistance(A=@n2.Transform.Translation, B=@n3.Transform.Translation).Result
   )
  ]}
 New Function_ContainedGraph {nodes [
   n1 = RigVMFunctionEntryNode()
   RigVMFunction_MathVectorLengthSquared = /Script/RigVM.RigVMFunction_MathVectorLengthSquared(Value=@n1.Pawn Velocity) {is=struct RigVMUnitNode, template="LengthSquared::Execute(in Value,out Result)"}
   n2 = &RigVMFunction_AlphaInterp(InterpSpeedDecreasing=10, InterpSpeedIncreasing=10, Scale=1, bClampResult=False, Value=@RigVMFunction_MathVectorLengthSquared.Result)
   > RigVMFunctionReturnNode(Is Moving=/Script/RigVM.RigVMFunction_MathDoubleGreater(B=0, A=@n2.Result) {is=struct RigVMUnitNode, template="Greater::Execute(in A,in B,out Result)"}.Result)
  ]}
 New Function_ContainedGraph {nodes [
   n1 = RigVMFunctionEntryNode()
   RigVMFunction_MathMatrixFromTransformV2 = /Script/RigVM.RigVMFunction_MathMatrixFromTransformV2(Value=@n1.Transform A) {is=struct RigVMUnitNode, template="Cast::Execute(in Value,out Result)"}
   RigVMFunction_MathMatrixToVectors = /Script/RigVM.RigVMFunction_MathMatrixToVectors(Value=@RigVMFunction_MathMatrixFromTransformV2.Result) {is=struct RigVMUnitNode}
   Interpolate = /Script/RigVM.RigVMFunction_MathVectorLerp(
    T=0.5
    A=@RigVMFunction_MathMatrixToVectors.Y
    B=&RigVMFunction_MathMatrixToVectors(Value=&RigVMFunction_MathMatrixFromTransformV2(Value=@n1.Transform C) {template="Cast::Execute(in Value,out Result)"}.Result).Y
   ) {is=struct RigVMUnitNode, template="Interpolate::Execute(in A,in B,in T,out Result)"}
   Unit = /Script/RigVM.RigVMFunction_MathVectorUnit(Value=@Interpolate.Result) {is=struct RigVMUnitNode, template="Unit::Execute(in Value,out Result)"}
   Scale = /Script/RigVM.RigVMFunction_MathVectorScale(Value=@Unit.Result, Factor=RigVMVariableNode() {variable=Distance}.Value) {is=struct RigVMUnitNode, template="Scale::Execute(in Value,in Factor,out Result)"}
   > RigVMFunctionReturnNode(
    Pole Vector=&RigVMFunction_MathVectorAdd(A=@Scale.Result, B=@n1.Transform B.Translation) {template="Add::Execute(in A,in B,out Result)"}.Result
    Root=RigVMVariableNode() {variable=Transform A}.Value
    Efector=RigVMVariableNode() {variable=Transform C}.Value
   )
  ]}
 New Function_ContainedGraph {nodes [
   n1 = RigVMFunctionEntryNode()
   RigVMFunction_MathVectorLength = /Script/RigVM.RigVMFunction_MathVectorLength(Value=@n1.Pawn Velocity) {is=struct RigVMUnitNode, template="Length::Execute(in Value,out Result)"}
   n2 = &RigVMFunction_AlphaInterp(InterpSpeedDecreasing=5, InterpSpeedIncreasing=5, Scale=1, bClampResult=False, Value=@RigVMFunction_MathVectorLength.Result)
   Less = /Script/RigVM.RigVMFunction_MathDoubleLess(A=@n2.Result, B=RigVMVariableNode() {variable=Velocity Threshold}.Value) {is=struct RigVMUnitNode, template="Less::Execute(in A,in B,out Result)"}
   > RigVMFunctionReturnNode(Is Not Running=@Less.Result)
  ]}
 New Function_ContainedGraph {nodes [
   n1 = RigVMFunctionEntryNode()
   Subtract = &RigVMFunction_MathDoubleSub(A=1, B=/Script/ControlRig.RigUnit_GetCurveValue(Curve=@n1.Curve Name) {is=struct RigVMUnitNode}.Value)
   > RigVMFunctionReturnNode(Final Alpha=RigVMDispatchNode(
     Condition=/Script/ControlRig.RigUnit_CurveExists(Curve=@n1.Curve Name) {is=struct RigVMUnitNode}.Exists
     True=&RigVMFunction_MathDoubleMul(A=@Subtract.Result, B=@n1.Alpha).Result
     False=@n1.Alpha
    ) {template="DISPATCH_RigVMDispatch_If(in Condition,in True,in False,out Result)"}.Result)
  ]}
 RigVMFunctionLibrary {nodes [
   Compute Alpha = RigVMCollapseNode(Alpha=1, Curve Name=DisableSlopeWarping)
   Leg Segment Lengths = RigVMCollapseNode(Calf=(Type=Bone,Name="None"), Foot=(Type=Bone,Name="None"), Thigh=(Type=Bone,Name="None"))
   Pawn is Moving = RigVMCollapseNode(Pawn Velocity=(X=0,Y=0,Z=0))
   Pawn is Not Running = RigVMCollapseNode(Pawn Velocity=(X=0,Y=0,Z=0), Velocity Threshold=500)
   Pre-IK Transforms = RigVMCollapseNode(
    Distance=10
    Transform A=(Rotation=(X=0,Y=0,Z=0,W=1),Translation=(X=0,Y=0,Z=0),Scale3D=(X=1,Y=1,Z=1))
    Transform B=(Rotation=(X=0,Y=0,Z=0,W=1),Translation=(X=0,Y=0,Z=0),Scale3D=(X=1,Y=1,Z=1))
    Transform C=(Rotation=(X=0,Y=0,Z=0,W=1),Translation=(X=0,Y=0,Z=0),Scale3D=(X=1,Y=1,Z=1))
   )
   Trace = RigVMCollapseNode(ExecuteContext=(), Foot Bone=(Type=Bone), Root Bone=(Type=Bone,Name="None"), Trace Distance=30, Trace Offset=0, Trace Radius=3)
  ]}
 RigVMFunction_Sequence_1_ContainedGraph {nodes [
   n1 = RigVMFunctionEntryNode()
   n2 = &RigVMFunction_Sequence(ExecuteContext=@n1.ExecuteContext)
   n3 = &RigVMFunction_Sequence(ExecuteContext=@n2.B)
   n4 = &RigVMFunction_Sequence(ExecuteContext=@n3.B)
   RigVMFunctionReturnNode(B=@n2.A, C=@n3.A, D=@n4.A, E=@n4.B)
  ]}
 RigVMFunction_Sequence_1_ContainedGraph {nodes [
   n1 = RigVMFunctionEntryNode()
   RigVMFunction_Sequence = &RigVMFunction_Sequence(ExecuteContext=@n1.ExecuteContext)
   n2 = &RigVMFunction_Sequence(ExecuteContext=@RigVMFunction_Sequence.B)
   n3 = &RigVMFunction_Sequence(ExecuteContext=@n2.B)
   RigVMFunctionReturnNode(A=@RigVMFunction_Sequence.A, B=@n2.A, C=@n3.A, D=@n3.B)
  ]}
 RigVMFunction_Sequence_3_ContainedGraph {nodes [
   n1 = RigVMFunctionEntryNode()
   n2 = &RigVMFunction_Sequence(ExecuteContext=@n1.ExecuteContext)
   n3 = &RigVMFunction_Sequence(ExecuteContext=@n2.B)
   RigVMFunctionReturnNode(B=@n3.A, C=@n3.B, A=@n2.A)
  ]}
 RigVMModel {nodes [
   RigUnit_BeginExecution = &RigUnit_BeginExecution()
   Compute Alpha = RigVMFunctionReferenceNode(Curve Name=DisableSlopeWarping, Alpha=RigVMVariableNode() {variable=Alpha}.Value)
   n1 = RigVMVariableNode(Value=@"Compute Alpha".Final Alpha, ExecuteContext=@RigUnit_BeginExecution.ExecutePin) {variable=Slope Alpha}
   Not_Equals = RigVMDispatchNode(B=0, A=RigVMVariableNode() {variable=Slope Alpha}.Value) {template="DISPATCH_RigVMDispatch_CoreNotEquals(in A,in B,out Result)"}
   RigVMFunction_MathQuaternionFromRotatorV2 = /Script/RigVM.RigVMFunction_MathQuaternionFromRotatorV2(Value=RigVMVariableNode() {variable=Slope Rotation}.Value) {is=struct RigVMUnitNode, template="Cast::Execute(in Value,out Result)"}
   ToVectors = /Script/RigVM.RigVMFunction_MathQuaternionToVectors(Value=@RigVMFunction_MathQuaternionFromRotatorV2.Result) {is=struct RigVMUnitNode, template="ToVectors::Execute(in Value,out Forward,out Right,out Up)"}
   Dot__ = /Script/RigVM.RigVMFunction_MathVectorDot(A=(X=0,Y=0,Z=1), B=@ToVectors.Up) {is=struct RigVMUnitNode, template="Dot,|::Execute(in A,in B,out Result)"}
   Less = &RigVMFunction_MathDoubleLess(B=1, A=@Dot__.Result) {template="Less::Execute(in A,in B,out Result)"}
   RigVMFunction_MathBoolOr = &RigVMFunction_MathBoolOr(
    B=@Not_Equals.Result
    A=&RigVMFunction_MathBoolAnd(B=@Less.Result, A=RigVMVariableNode() {variable=Use Slope Info Input}.Value).Result
   )
   n2 = &RigVMFunction_ControlFlowBranch(Condition=@RigVMFunction_MathBoolOr.Result, ExecuteContext=@n1.ExecuteContext)
   RigVMFunction_Sequence = RigVMAggregateNode(ExecuteContext=@n2.True)
   n3 = &RigUnit_GetTransform(bInitial=False, Item=RigVMVariableNode() {variable=Foot L}.Value)
   Add = RigVMAggregateNode(
    A=@n3.Transform.Translation.Z
    B=RigVMVariableNode() {variable=Offset Location Z Foot L}.Value
    C=RigVMVariableNode() {variable=Floor Offset}.Value
   )
   Multiply = /Script/RigVM.RigVMFunction_MathQuaternionMul(B=@n3.Transform.Rotation, A=RigVMVariableNode() {variable=Offset Rotation Foot L}.Value) {is=struct RigVMUnitNode, template="Multiply::Execute(in A,in B,out Result)"}
   RigVMVariableNode(
    Value=(Rotation=(X=0,Y=0,Z=0,W=1),Translation=(X=0,Y=0,Z=0),Scale3D=(X=1,Y=1,Z=1))
    Value.Translation.Z=@Add.Result
    Value.Scale3D=@n3.Transform.Scale3D
    Value.Translation.X=@n3.Transform.Translation.X
    Value.Translation.Y=@n3.Transform.Translation.Y
    Value.Rotation=@Multiply.Result
    ExecuteContext=@RigVMFunction_Sequence.D
   ) {variable=Transform Foot L}
   n4 = &RigUnit_GetTransform(bInitial=False, Item=RigVMVariableNode() {variable=Foot R}.Value)
   Add_1 = RigVMAggregateNode(
    A=@n4.Transform.Translation.Z
    C=RigVMVariableNode() {variable=Floor Offset}.Value
    B=RigVMVariableNode() {variable=Offset Location Z Foot R}.Value
   )
   > RigVMVariableNode(
    Value=(Rotation=(X=0,Y=0,Z=0,W=1),Translation=(X=0,Y=0,Z=0),Scale3D=(X=1,Y=1,Z=1))
    Value.Translation.Z=@Add_1.Result
    Value.Scale3D=@n4.Transform.Scale3D
    Value.Translation.X=@n4.Transform.Translation.X
    Value.Translation.Y=@n4.Transform.Translation.Y
    Value.Rotation=&RigVMFunction_MathQuaternionMul(B=@n4.Transform.Rotation, A=RigVMVariableNode() {variable=Offset Rotation Foot R}.Value) {template="Multiply::Execute(in A,in B,out Result)"}.Result
   ) {variable=Transform Foot R}
   &RigUnit_ModifyTransforms(
    ItemToModify=((Item=(Type=Bone,Name="pelvis"),Transform=(Rotation=(X=0,Y=0,Z=0,W=1),Translation=(X=0,Y=0,Z=0),Scale3D=(X=1,Y=1,Z=1))))
    Mode=AdditiveGlobal
    ItemToModify.0.Transform.Translation.Z=&RigVMFunction_MathDoubleAdd(B=RigVMVariableNode() {variable=Floor Offset}.Value, A=RigVMVariableNode() {variable=Offset Location Z Pelvis}.Value).Result
    ItemToModify.0.Item=RigVMVariableNode() {variable=Pelvis}.Value
    Weight=RigVMVariableNode() {variable=Slope Alpha}.Value
    ExecutePin=@RigVMFunction_Sequence.E
   )
   Pre-IK Transforms = RigVMFunctionReferenceNode(
    Distance=10
    Transform A=&RigUnit_GetTransform(bInitial=False, Item=RigVMVariableNode() {variable=Thigh L}.Value).Transform
    Transform B=&RigUnit_GetTransform(bInitial=False, Item=RigVMVariableNode() {variable=Calf L}.Value).Transform
    Transform C=RigVMVariableNode() {variable=Transform Foot L}.Value
   )
   n5 = /Script/ControlRig.RigUnit_TwoBoneIKSimpleTransforms(
    Elbow=(Rotation=(X=0,Y=0,Z=0,W=1),Translation=(X=0,Y=0,Z=0),Scale3D=(X=1,Y=1,Z=1))
    PrimaryAxis=(X=-1,Y=0,Z=0)
    SecondaryAxis=(X=0,Y=1,Z=0)
    SecondaryAxisWeight=1
    StretchMaximumRatio=1.25
    StretchStartRatio=0.75
    bEnableStretch=false
    Effector=@"Pre-IK Transforms".Efector
    PoleVector=@"Pre-IK Transforms".Pole Vector
    Root=@"Pre-IK Transforms".Root
    BoneALength=RigVMVariableNode() {variable=Bone A Length L}.Value
    BoneBLength=RigVMVariableNode() {variable=Bone B Length L}.Value
   ) {is=struct RigVMUnitNode}
   > /Script/ControlRig.RigUnit_SetTransformItemArray(
    Items=((Type=Bone,Name="thigh_l"),(Type=Bone,Name="calf_l"),(Type=Bone,Name="foot_l"))
    Space=GlobalSpace
    Transforms=((Rotation=(X=0,Y=0,Z=0,W=1),Translation=(X=0,Y=0,Z=0),Scale3D=(X=1,Y=1,Z=1)),(Rotation=(X=0,Y=0,Z=0,W=1),Translation=(X=0,Y=0,Z=0),Scale3D=(X=1,Y=1,Z=1)),(Rotation=(X=0,Y=0,Z=0,W=1),Translation=(X=0,Y=0,Z=0),Scale3D=(X=1,Y=1,Z=1)))
    bInitial=False
    bPropagateToChildren=True
    Transforms.2=@n5.Effector
    Transforms.1=@n5.Elbow
    Transforms.0=@n5.Root
    Items.0=RigVMVariableNode() {variable=Thigh L}.Value
    Items.1=RigVMVariableNode() {variable=Calf L}.Value
    Items.2=RigVMVariableNode() {variable=Foot L}.Value
    Weight=RigVMVariableNode() {variable=Slope Alpha}.Value
   ) {is=struct RigVMUnitNode}
   Pre-IK Transforms_1 = RigVMFunctionReferenceNode(
    Distance=-10
    Transform A=&RigUnit_GetTransform(bInitial=False, Item=RigVMVariableNode() {variable=Thigh R}.Value).Transform
    Transform B=&RigUnit_GetTransform(bInitial=False, Item=RigVMVariableNode() {variable=Calf R}.Value).Transform
    Transform C=RigVMVariableNode() {variable=Transform Foot R}.Value
   )
   n6 = &RigUnit_TwoBoneIKSimpleTransforms(
    Elbow=(Rotation=(X=0,Y=0,Z=0,W=1),Translation=(X=0,Y=0,Z=0),Scale3D=(X=1,Y=1,Z=1))
    PrimaryAxis=(X=1,Y=0,Z=0)
    SecondaryAxis=(X=0,Y=-1,Z=0)
    SecondaryAxisWeight=1
    StretchMaximumRatio=1.25
    StretchStartRatio=0.75
    bEnableStretch=false
    Effector=@"Pre-IK Transforms_1".Efector
    PoleVector=@"Pre-IK Transforms_1".Pole Vector
    Root=@"Pre-IK Transforms_1".Root
    BoneALength=RigVMVariableNode() {variable=Bone A Length R}.Value
    BoneBLength=RigVMVariableNode() {variable=Bone B Length R}.Value
   )
   > &RigUnit_SetTransformItemArray(
    Items=((Type=Bone,Name="thigh_r"),(Type=Bone,Name="calf_r"),(Type=Bone,Name="foot_r"))
    Space=GlobalSpace
    Transforms=((Rotation=(X=0,Y=0,Z=0,W=1),Translation=(X=0,Y=0,Z=0),Scale3D=(X=1,Y=1,Z=1)),(Rotation=(X=0,Y=0,Z=0,W=1),Translation=(X=0,Y=0,Z=0),Scale3D=(X=1,Y=1,Z=1)),(Rotation=(X=0,Y=0,Z=0,W=1),Translation=(X=0,Y=0,Z=0),Scale3D=(X=1,Y=1,Z=1)))
    bInitial=False
    bPropagateToChildren=True
    Transforms.2=@n6.Effector
    Transforms.1=@n6.Elbow
    Transforms.0=@n6.Root
    Items.0=RigVMVariableNode() {variable=Thigh R}.Value
    Items.1=RigVMVariableNode() {variable=Calf R}.Value
    Items.2=RigVMVariableNode() {variable=Foot R}.Value
    Weight=RigVMVariableNode() {variable=Slope Alpha}.Value
   )
   Pawn is Not Running = RigVMFunctionReferenceNode(Pawn Velocity=RigVMVariableNode() {variable=Pawn Velocity}.Value, Velocity Threshold=RigVMVariableNode() {variable=Velocity Threshold}.Value)
   RigVMFunction_MathBoolNot = &RigVMFunction_MathBoolNot(Value=RigVMVariableNode() {variable=Is In Air}.Value)
   RigVMFunction_MathBoolAnd = &RigVMFunction_MathBoolAnd(B=@"Pawn is Not Running".Is Not Running, A=@RigVMFunction_MathBoolNot.Result)
   Branch_RigVMUnitNode = &RigVMFunction_ControlFlowBranch(Condition=@RigVMFunction_MathBoolAnd.Result, ExecuteContext=@RigVMFunction_Sequence.B)
   RigVMVariableNode(ExecuteContext=@Branch_RigVMUnitNode.False) {variable=Hit Location Z Foot L}
   > RigVMVariableNode(Value=()) {variable=Hit Rotation Foot L}
   > RigVMVariableNode() {variable=Hit Location Z Foot R}
   > RigVMVariableNode(Value=()) {variable=Hit Rotation Foot R}
   Pawn is Moving = RigVMFunctionReferenceNode(Pawn Velocity=RigVMVariableNode() {variable=Pawn Velocity}.Value)
   RigVMFunction_ControlFlowBranch = &RigVMFunction_ControlFlowBranch(ExecuteContext=@Branch_RigVMUnitNode.True, Condition=@"Pawn is Moving".Is Moving)
   n7 = &RigVMFunction_ControlFlowBranch(ExecuteContext=@RigVMFunction_ControlFlowBranch.True, Condition=RigVMVariableNode() {variable=Single Trace}.Value)
   Trace_1 = RigVMFunctionReferenceNode(
    Trace Distance=30
    Trace Offset=0
    Trace Radius=3
    ExecuteContext=@n7.False
    Foot Bone=RigVMVariableNode() {variable=Foot L}.Value
    Root Bone=RigVMVariableNode() {variable=Root}.Value
    Debug Draw=RigVMVariableNode() {variable=Debug Draw}.Value
   )
   > RigVMVariableNode(Value=@Trace_1.Hit Position.Z) {variable=Hit Location Z Foot L}
   > RigVMVariableNode(Value=@Trace_1.Hit Rotation) {variable=Hit Rotation Foot L}
   > Trace_2 = RigVMFunctionReferenceNode(
    Trace Distance=30
    Trace Offset=0
    Trace Radius=3
    Foot Bone=RigVMVariableNode() {variable=Foot R}.Value
    Root Bone=RigVMVariableNode() {variable=Root}.Value
    Debug Draw=RigVMVariableNode() {variable=Debug Draw}.Value
   )
   > RigVMVariableNode(Value=@Trace_2.Hit Position.Z) {variable=Hit Location Z Foot R}
   > RigVMVariableNode(Value=@Trace_2.Hit Rotation) {variable=Hit Rotation Foot R}
   n8 = &RigVMFunction_ControlFlowBranch(ExecuteContext=@n7.True, Condition=RigVMVariableNode() {variable=Use Slope Info Input}.Value)
   Trace = RigVMFunctionReferenceNode(
    Trace Distance=30
    Trace Offset=0
    Trace Radius=3
    ExecuteContext=@n8.False
    Foot Bone=RigVMVariableNode() {variable=Pelvis}.Value
    Root Bone=RigVMVariableNode() {variable=Root}.Value
    Debug Draw=RigVMVariableNode() {variable=Debug Draw}.Value
   )
   > RigVMVariableNode(Value=@Trace.Hit Position) {variable=Hit Location Root}
   > RigVMVariableNode(Value=@Trace.Hit Position.Z) {variable=Hit Location Z Root}
   > RigVMVariableNode(Value=@Trace.Hit Normal) {variable=Hit Normal Root}
   > RigVMVariableNode(Value=@Trace.Hit Rotation) {variable=Hit Rotation Foot L}
   > RigVMVariableNode(Value=@Trace.Hit Rotation) {variable=Hit Rotation Foot R}
   n9 = &RigUnit_GetTransform(Item=(Type=Bone,Name="foot_l"), bInitial=False)
   n10 = RigVMDispatchNode(
    False=1
    True=-1
    Condition=&RigVMFunction_MathDoubleLess(B=@n9.Transform.Translation.Z, A=RigVMVariableNode() {variable=Hit Location Z Root}.Value) {template="Less::Execute(in A,in B,out Result)"}.Result
   ) {template="DISPATCH_RigVMDispatch_If(in Condition,in True,in False,out Result)"}
   RigVMVariableNode(
    Value=/Script/RigVM.RigVMFunction_MathIntersectPlane(
     PlanePoint=RigVMVariableNode() {variable=Hit Location Root}.Value
     PlaneNormal=RigVMVariableNode() {variable=Hit Normal Root}.Value
     Start=@n9.Transform.Translation
     Direction=RigVMDispatchNode(Value=(X=0,Y=0,Z=1), Value.Z=@n10.Result) {template="DISPATCH_RigVMDispatch_Constant(io Value)"}.Value
    ) {is=struct RigVMUnitNode}.Result.Z
    ExecuteContext=@n8.Completed
   ) {variable=Hit Location Z Foot L}
   n11 = &RigUnit_GetTransform(Item=(Type=Bone,Name="foot_r"), bInitial=False)
   n12 = RigVMDispatchNode(
    False=1
    True=-1
    Condition=&RigVMFunction_MathDoubleLess(B=@n11.Transform.Translation.Z, A=RigVMVariableNode() {variable=Hit Location Z Root}.Value) {template="Less::Execute(in A,in B,out Result)"}.Result
   ) {template="DISPATCH_RigVMDispatch_If(in Condition,in True,in False,out Result)"}
   > RigVMVariableNode(Value=&RigVMFunction_MathIntersectPlane(
     PlanePoint=RigVMVariableNode() {variable=Hit Location Root}.Value
     PlaneNormal=RigVMVariableNode() {variable=Hit Normal Root}.Value
     Start=@n11.Transform.Translation
     Direction=RigVMDispatchNode(Value=(X=0,Y=0,Z=1), Value.Z=@n12.Result) {template="DISPATCH_RigVMDispatch_Constant(io Value)"}.Value
    ).Result.Z) {variable=Hit Location Z Foot R}
   From World = /Script/ControlRig.RigUnit_ToRigSpace_Location(Value=RigVMVariableNode() {variable=Root Slope Translation}.Value) {is=struct RigVMUnitNode, template="From World::Execute(in Value,out Global)"}
   RigVMVariableNode(Value=@"From World".Global, ExecuteContext=@n8.True) {variable=Hit Location Root}
   > RigVMVariableNode(Value=@"From World".Global.Z) {variable=Hit Location Z Root}
   From World_1 = /Script/ControlRig.RigUnit_ToRigSpace_Rotation(Value=&RigVMFunction_MathQuaternionFromRotatorV2(Value=RigVMVariableNode() {variable=Slope Rotation}.Value) {template="Cast::Execute(in Value,out Result)"}.Result) {is=struct RigVMUnitNode, template="From World::Execute(in Value,out Global)"}
   > RigVMVariableNode(Value=&RigVMFunction_MathQuaternionToVectors(Value=@"From World_1".Global) {template="ToVectors::Execute(in Value,out Forward,out Right,out Up)"}.Up) {variable=Hit Normal Root}
   > RigVMVariableNode(Value=@"From World_1".Global) {variable=Hit Rotation Foot L}
   > RigVMVariableNode(Value=@"From World_1".Global) {variable=Hit Rotation Foot R}
   AlphaInterp = &RigVMFunction_AlphaInterp(
    InRange=(Max=1)
    InterpSpeedDecreasing=15
    InterpSpeedIncreasing=15
    OutRange=(Max=1)
    Scale=1
    Bias=~
    ClampMin=~
    bMapRange=~
    Value=RigVMVariableNode() {variable=Hit Location Z Foot L}.Value
   )
   RigVMVariableNode(Value=@AlphaInterp.Result, ExecuteContext=@RigVMFunction_Sequence.C) {variable=Offset Location Z Foot L}
   n13 = &RigVMFunction_AlphaInterp(
    InRange=(Max=1)
    InterpSpeedDecreasing=15
    InterpSpeedIncreasing=15
    OutRange=(Max=1)
    Scale=1
    Bias=~
    ClampMin=~
    bMapRange=~
    Value=RigVMVariableNode() {variable=Hit Location Z Foot R}.Value
   )
   > RigVMVariableNode(Value=@n13.Result) {variable=Offset Location Z Foot R}
   n14 = /Script/RigVM.RigVMFunction_AlphaInterpQuat(
    Bias=0
    ClampMax=1
    ClampMin=0
    InRange=(Min=0,Max=1)
    InterpSpeedDecreasing=10
    InterpSpeedIncreasing=10
    OutRange=(Min=0,Max=1)
    Scale=1
    bClampResult=False
    bInterpResult=False
    bMapRange=False
    Value=RigVMVariableNode() {variable=Hit Rotation Foot L}.Value
   ) {
    is=struct RigVMUnitNode
    template="AlphaInterp::Execute(in Value,in Scale,in Bias,in bMapRange,in InRange,in OutRange,in bClampResult,in ClampMin,in ClampMax,in bInterpResult,in InterpSpeedIncreasing,in InterpSpeedDecreasing,out Result)"
   }
   > RigVMVariableNode(Value=@n14.Result) {variable=Offset Rotation Foot L}
   n15 = &RigVMFunction_AlphaInterpQuat(
    Bias=0
    ClampMax=1
    ClampMin=0
    InRange=(Min=0,Max=1)
    InterpSpeedDecreasing=10
    InterpSpeedIncreasing=10
    OutRange=(Min=0,Max=1)
    Scale=1
    bClampResult=False
    bInterpResult=False
    bMapRange=False
    Value=RigVMVariableNode() {variable=Hit Rotation Foot R}.Value
   ) {template="AlphaInterp::Execute(in Value,in Scale,in Bias,in bMapRange,in InRange,in OutRange,in bClampResult,in ClampMin,in ClampMax,in bInterpResult,in InterpSpeedIncreasing,in InterpSpeedDecreasing,out Result)"}
   > RigVMVariableNode(Value=@n15.Result) {variable=Offset Rotation Foot R}
   n16 = RigVMVariableNode() {variable=Offset Location Z Foot L}
   n17 = RigVMVariableNode() {variable=Offset Location Z Foot R}
   > RigVMVariableNode(Value=RigVMDispatchNode(
     Condition=/Script/RigVM.RigVMFunction_MathFloatLess(A=@n16.Value, B=@n17.Value) {is=struct RigVMUnitNode, template="Less::Execute(in A,in B,out Result)"}.Result
     True=RigVMVariableNode() {variable=Offset Location Z Foot L}.Value
     False=RigVMVariableNode() {variable=Offset Location Z Foot R}.Value
    ) {template="DISPATCH_RigVMDispatch_If(in Condition,in True,in False,out Result)"}.Result) {variable=Offset Location Z Pelvis}
  ]}
]
hierarchy [
 Root {type=Connector, loc=<0 0 0>, rot=<0 0 0 1>}
 Root {type=Socket, loc=<0 0 0>, rot=<0 0 0 1>}
 Pelvis {type=Connector, loc=<0 0 0>, rot=<0 0 0 1>}
 Foot L {type=Connector, loc=<0 0 0>, rot=<0 0 0 1>}
 Foot R {type=Connector, loc=<0 0 0>, rot=<0 0 0 1>}
]
variables [
 Alpha {type=double, default=1}
 Bone A Length L {type=double, default=0}
 Bone A Length R {type=double, default=0}
 Bone B Length L {type=double, default=0}
 Bone B Length R {type=double, default=0}
 Calf L {type=FRigElementKey, default=(Type=None,Name="")}
 Calf R {type=FRigElementKey, default=(Type=None,Name="")}
 Debug Draw {type=bool, default=False}
 Floor Offset {type=double, default=0}
 Foot L {type=FRigElementKey, default=(Type=None,Name="")}
 Foot R {type=FRigElementKey, default=(Type=None,Name="")}
 Height Foot L {type=double, default=0}
 Height Foot R {type=double, default=0}
 Hit Location Root {type=FVector, default=(X=0,Y=0,Z=0)}
 Hit Location Z Foot L {type=double, default=0}
 Hit Location Z Foot R {type=double, default=0}
 Hit Location Z Root {type=double, default=0}
 Hit Normal Root {type=FVector, default=(X=0,Y=0,Z=0)}
 Hit Rotation Foot L {type=FQuat, default=(X=0,Y=0,Z=0,W=1)}
 Hit Rotation Foot R {type=FQuat, default=(X=0,Y=0,Z=0,W=1)}
 Hit Rotation Root {type=FQuat, default=(X=0,Y=0,Z=0,W=1)}
 Is In Air {type=bool, default=False}
 Offset Location Z Foot L {type=double, default=0}
 Offset Location Z Foot R {type=double, default=0}
 Offset Location Z Pelvis {type=double, default=0}
 Offset Rotation Foot L {type=FQuat, default=(X=0,Y=0,Z=0,W=1)}
 Offset Rotation Foot R {type=FQuat, default=(X=0,Y=0,Z=0,W=1)}
 Pawn Velocity {type=FVector, default=(X=0,Y=0,Z=0)}
 Pelvis {type=FRigElementKey, default=(Type=None,Name="")}
 Root {type=FRigElementKey, default=(Type=None,Name="")}
 Root Slope Translation {type=FVector, default=(X=0,Y=0,Z=0)}
 Single Trace {type=bool, default=True}
 Slope Alpha {type=double, default=0}
 Slope Rotation {type=FRotator, default=(Pitch=0,Yaw=0,Roll=0)}
 Thigh L {type=FRigElementKey, default=(Type=None,Name="")}
 Thigh R {type=FRigElementKey, default=(Type=None,Name="")}
 Transform Foot L {type=FTransform, default=(Rotation=(X=0,Y=0,Z=0,W=1),Translation=(X=0,Y=0,Z=0),Scale3D=(X=1,Y=1,Z=1))}
 Transform Foot R {type=FTransform, default=(Rotation=(X=0,Y=0,Z=0,W=1),Translation=(X=0,Y=0,Z=0),Scale3D=(X=1,Y=1,Z=1))}
 Use Slope Info Input {type=bool, default=True}
 Velocity Threshold {type=double, default=500}
]
