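# Control Rig module that runs a spherical pose reader on a connected driver, with optional parent, and drives items and curves from the resulting weight.
# The reader's output is stored and optionally published as module metadata.
# Connected Driven Items receive index-matched transforms, clamped to the last entry, weighted by that output in the selected local/global, override/additive mode.
# Curves matched by partial name receive the weight remapped into per-entry ranges.
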ControlRigBlueprint /EpicControlRigRuntime/CRM_Runtime_SphericalReader_v01
graphs [
 RigVMFunctionLibrary {links [], nodes []}
 RigVMFunction_Sequence_1_ContainedGraph {nodes [
   n1 = RigVMFunctionEntryNode()
   RigVMFunction_Sequence = &RigVMFunction_Sequence(ExecuteContext=@n1.ExecuteContext)
   n2 = &RigVMFunction_Sequence(ExecuteContext=@RigVMFunction_Sequence.B)
   RigVMFunctionReturnNode(A=@RigVMFunction_Sequence.A, B=@n2.A, C=@n2.B)
  ]}
 RigVMModel {nodes [
   RigUnit_BeginExecution = &RigUnit_BeginExecution()
   RigVMFunction_Sequence = RigVMAggregateNode(ExecuteContext=@RigUnit_BeginExecution.ExecutePin)
   n1 = /Script/ControlRig.RigUnit_ResolveArrayConnector(Connector=(Type=Connector,Name="Driven Items"), SkipSocket=False) {is=struct RigVMUnitNode}
   RigVMFunction_ControlFlowBranch = &RigVMFunction_ControlFlowBranch(Condition=@n1.bIsConnected, ExecuteContext=@RigVMFunction_Sequence.B)
   For_Each_3 = RigVMDispatchNode(Array=@n1.Result, ExecuteContext=@RigVMFunction_ControlFlowBranch.True) {template="DISPATCH_RigVMDispatch_ArrayIterator(in Array,out Element,out Index,out Count,out Ratio)"}
   > RigVMVariableNode(Value=@For_Each_3.Element) {variable=Current Item}
   n2 = RigVMVariableNode() {variable=Driven Transforms}
   Subtract = /Script/RigVM.RigVMFunction_MathIntSub(B=1, A=RigVMDispatchNode(Array=@n2.Value) {template="DISPATCH_RigVMDispatch_ArrayGetNum(in Array,out Num)"}.Num) {is=struct RigVMUnitNode, template="Subtract::Execute(in A,in B,out Result)"}
   Clamp = &RigVMFunction_MathIntClamp(Value=@For_Each_3.Index, Maximum=@Subtract.Result)
   > RigVMVariableNode(Value=RigVMDispatchNode(Index=@Clamp.Result, Array=@n2.Value) {template="DISPATCH_RigVMDispatch_ArrayGetAtIndex(in Array,in Index,out Element)"}.Element) {variable=Current Transform}
   > n3 = RigVMDispatchNode(Index=RigVMDispatchNode(Value=RigVMVariableNode() {variable=Driven Mode}.Value) {template="DISPATCH_RigVMDispatch_CastEnumToInt(in Value,out Result)"}.Result) {template="DISPATCH_RigVMDispatch_SwitchInt32(in Index)"}
   &RigUnit_ModifyTransforms(
    ItemToModify=((Item=(Type=Bone,Name="None"),Transform=(Rotation=(X=0,Y=0,Z=0,W=1),Translation=(X=0,Y=0,Z=0),Scale3D=(X=1,Y=1,Z=1))))
    Mode=AdditiveGlobal
    WeightMinimum=0
    Weight=RigVMVariableNode() {variable=Output Param}.Value
    ItemToModify.0.Item=RigVMVariableNode() {variable=Current Item}.Value
    ItemToModify.0.Transform=RigVMVariableNode() {variable=Current Transform}.Value
    ExecutePin=@n3.Cases.3
   )
   &RigUnit_ModifyTransforms(
    ItemToModify=((Item=(Type=Bone,Name="None"),Transform=(Rotation=(X=0,Y=0,Z=0,W=1),Translation=(X=0,Y=0,Z=0),Scale3D=(X=1,Y=1,Z=1))))
    WeightMinimum=0
    Weight=RigVMVariableNode() {variable=Output Param}.Value
    ItemToModify.0.Item=RigVMVariableNode() {variable=Current Item}.Value
    ItemToModify.0.Transform=RigVMVariableNode() {variable=Current Transform}.Value
    ExecutePin=@n3.Cases.2
   )
   &RigUnit_ModifyTransforms(
    ItemToModify=((Item=(Type=Bone,Name="None"),Transform=(Rotation=(X=0,Y=0,Z=0,W=1),Translation=(X=0,Y=0,Z=0),Scale3D=(X=1,Y=1,Z=1))))
    Mode=OverrideGlobal
    WeightMinimum=0
    Weight=RigVMVariableNode() {variable=Output Param}.Value
    ItemToModify.0.Item=RigVMVariableNode() {variable=Current Item}.Value
    ItemToModify.0.Transform=RigVMVariableNode() {variable=Current Transform}.Value
    ExecutePin=@n3.Cases.1
   )
   &RigUnit_ModifyTransforms(
    ItemToModify=((Item=(Type=Bone,Name="None"),Transform=(Rotation=(X=0,Y=0,Z=0,W=1),Translation=(X=0,Y=0,Z=0),Scale3D=(X=1,Y=1,Z=1))))
    Mode=OverrideLocal
    WeightMinimum=0
    Weight=RigVMVariableNode() {variable=Output Param}.Value
    ItemToModify.0.Item=RigVMVariableNode() {variable=Current Item}.Value
    ItemToModify.0.Transform=RigVMVariableNode() {variable=Current Transform}.Value
    ExecutePin=@n3.Cases.0
   )
   n4 = &RigUnit_SphericalPoseReader(
    DriverItem=/Script/ControlRig.RigUnit_ResolveConnector(Connector=(Type=Connector,Name="Driver Item"), SkipSocket=true) {is=struct RigVMUnitNode}.Result
    OptionalParentItem=&RigUnit_ResolveConnector(Connector=(Type=Connector,Name="Optional Parent"), SkipSocket=true).Result
    DriverAxis=RigVMVariableNode() {variable=DriverAxis}.Value
    RotationOffset=RigVMVariableNode() {variable=RotationOffset}.Value
    ActiveRegionSize=RigVMVariableNode() {variable=ActiveRegionSize}.Value
    ActiveRegionScaleFactors=RigVMVariableNode() {variable=ActiveRegionScaleFactors}.Value
    FalloffSize=RigVMVariableNode() {variable=FalloffSize}.Value
    FalloffRegionScaleFactors=RigVMVariableNode() {variable=FalloffRegionScaleFactors}.Value
    FlipWidthScaling=RigVMVariableNode() {variable=FlipWidthScaling}.Value
    FlipHeightScaling=RigVMVariableNode() {variable=FlipHeightScaling}.Value
    Debug=RigVMVariableNode() {variable=Debug}.Value
    ExecutePin=@RigVMFunction_Sequence.A
   )
   > n5 = RigVMVariableNode(Value=@n4.OutputParam) {variable=Output Param}
   n6 = &RigVMFunction_ControlFlowBranch(ExecuteContext=@n5.ExecuteContext, Condition=RigVMVariableNode() {variable=Set Weight Metadata}.Value)
   Set_Module_Metadata = RigVMDispatchNode(
    ExecuteContext=@n6.True
    Name=RigVMVariableNode() {variable=Metadata Name}.Value
    Value=RigVMVariableNode() {variable=Output Param}.Value
    NameSpace=RigVMVariableNode() {variable=Metadata Namespace}.Value
   ) {template="DISPATCH_RigDispatch_SetModuleMetadata(in Name,in NameSpace,in Value,out Success)"}
   For_Each = RigVMDispatchNode(Array=RigVMVariableNode() {variable=Driven Curves}.Value, ExecuteContext=@RigVMFunction_Sequence.C) {template="DISPATCH_RigVMDispatch_ArrayIterator(in Array,out Element,out Index,out Count,out Ratio)"}
   > For_Each_2 = RigVMDispatchNode(Array=/Script/ControlRig.RigUnit_CollectionNameSearchArray(TypeToSearch=Curve, PartialName=@For_Each.Element.Control) {is=struct RigVMUnitNode}.Items) {template="DISPATCH_RigVMDispatch_ArrayIterator(in Array,out Element,out Index,out Count,out Ratio)"}
   > /Script/ControlRig.RigUnit_SetCurveValue(
    Curve=@For_Each_2.Element.Name
    Value=&RigVMFunction_MathDoubleRemap(
     SourceMaximum=1
     SourceMinimum=0
     bClamp=False
     TargetMinimum=@For_Each.Element.Vector.X
     TargetMaximum=@For_Each.Element.Vector.Y
     Value=RigVMVariableNode() {variable=Output Param}.Value
    ).Result
   ) {is=struct RigVMUnitNode}
  ]}
]
hierarchy [
 Driver Item {type=Connector, loc=<0 0 0>, rot=<0 0 0 1>}
 Root {type=Socket, loc=<0 -2.3887 106.4931>, rot=<0 0.7071 0 -0.7071>}
 Optional Parent {type=Connector, loc=<0 0 0>, rot=<0 0 0 1>}
 Driven Items {type=Connector, loc=<0 0 0>, rot=<0 0 0 1>}
]
variables [
 ActiveRegionScaleFactors {type=FRegionScaleFactors, default=(PositiveWidth=1,NegativeWidth=1,PositiveHeight=1,NegativeHeight=1)}
 ActiveRegionSize {type=float, default=0.1}
 Current Item {type=FRigElementKey, default=(Type=Bone,Name="")}
 Current Transform {type=FTransform, default=(Rotation=(X=0,Y=0,Z=0,W=1),Translation=(X=0,Y=0,Z=0),Scale3D=(X=1,Y=1,Z=1))}
 Debug {
  type=FSphericalPoseReaderDebugSettings
  default=(bDrawDebug=True,bDraw2D=False,bDrawLocalAxes=False,DebugScale=25,DebugSegments=20,DebugThickness=0.25)
 }
 Driven Curves {type="TArray<FRigUnit_SetMultiControlVector2D_Entry>"}
 Driven Mode {type=EControlRigModifyBoneMode, default=AdditiveLocal}
 Driven Transforms {type="TArray<FTransform>"}
 DriverAxis {type=FVector, default=(X=1,Y=0,Z=0)}
 FalloffRegionScaleFactors {type=FRegionScaleFactors, default=(PositiveWidth=1,NegativeWidth=1,PositiveHeight=1,NegativeHeight=1)}
 FalloffSize {type=float, default=0.2}
 FlipHeightScaling {type=bool, default=False}
 FlipWidthScaling {type=bool, default=False}
 Metadata Name {type=FName, default=MySphericalReaderWeights}
 Metadata Namespace {type=ERigMetaDataNameSpace, default=None}
 Output Param {type=double, default=0}
 RotationOffset {type=FVector, default=(X=90,Y=0,Z=-90)}
 Set Weight Metadata {type=bool, default=False}
]
